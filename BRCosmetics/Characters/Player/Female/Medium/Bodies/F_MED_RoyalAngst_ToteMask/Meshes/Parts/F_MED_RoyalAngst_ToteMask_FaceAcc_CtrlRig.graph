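# Control Rig for the RoyalAngst ToteMask female face accessory that procedurally poses three dynamic hair chains (L, M, R) around the body.
# Hair controls follow weighted parent constraints to head, spine, shoulders and pelvis, take head and neck rotation correctives, then drive bones through B-spline IK.
# Further bone offsets are weighted by swing/twist delta rotations, spherical pose readers and gameplay inputs like aiming, dive-floating and back bling.
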ControlRigBlueprint /BRCosmetics/Characters/Player/Female/Medium/Bodies/F_MED_RoyalAngst_ToteMask/Meshes/Parts/F_MED_RoyalAngst_ToteMask_FaceAcc_CtrlRig
graphs [
 Multiply_13_ContainedGraph {nodes [
   n1 = RigVMFunctionEntryNode()
   n2 = &RigVMFunction_MathDoubleMul(A=@n1.A, B=@n1.C)
   RigVMFunctionReturnNode(Result=&RigVMFunction_MathDoubleMul(B=@n1.E, A=&RigVMFunction_MathDoubleMul(B=@n1.D, A=@n2.Result).Result).Result)
  ]}
 Multiply_1_ContainedGraph {nodes [
   n1 = RigVMFunctionEntryNode()
   RigVMFunctionReturnNode(Result=&RigVMFunction_MathDoubleMul(B=@n1.C, A=&RigVMFunction_MathDoubleMul(A=@n1.A, B=@n1.B).Result).Result)
  ]}
 Multiply_1_ContainedGraph {nodes [
   n1 = RigVMFunctionEntryNode()
   RigVMFunctionReturnNode(Result=&RigVMFunction_MathDoubleMul(B=@n1.C, A=&RigVMFunction_MathDoubleMul(A=@n1.A, B=@n1.B).Result).Result)
  ]}
 Multiply_1_ContainedGraph {nodes [
   n1 = RigVMFunctionEntryNode()
   RigVMFunctionReturnNode(Result=&RigVMFunction_MathDoubleMul(B=@n1.C, A=&RigVMFunction_MathDoubleMul(A=@n1.A, B=@n1.B).Result).Result)
  ]}
 Multiply_1_ContainedGraph {nodes [
   n1 = RigVMFunctionEntryNode()
   n2 = &RigVMFunction_MathDoubleMul(A=@n1.A, B=@n1.B)
   RigVMFunctionReturnNode(Result=&RigVMFunction_MathDoubleMul(B=@n1.D, A=&RigVMFunction_MathDoubleMul(B=@n1.C, A=@n2.Result).Result).Result)
  ]}
 Multiply_1_ContainedGraph {nodes [
   n1 = RigVMFunctionEntryNode()
   n2 = &RigVMFunction_MathDoubleMul(A=@n1.A, B=@n1.B)
   RigVMFunctionReturnNode(Result=&RigVMFunction_MathDoubleMul(B=@n1.D, A=&RigVMFunction_MathDoubleMul(B=@n1.C, A=@n2.Result).Result).Result)
  ]}
 Multiply_1_ContainedGraph {nodes [
   n1 = RigVMFunctionEntryNode()
   RigVMFunctionReturnNode(Result=&RigVMFunction_MathDoubleMul(B=@n1.D, A=&RigVMFunction_MathDoubleMul(A=@n1.A, B=@n1.C).Result).Result)
  ]}
 Multiply_1_ContainedGraph {nodes [
   n1 = RigVMFunctionEntryNode()
   RigVMFunctionReturnNode(Result=&RigVMFunction_MathDoubleMul(B=@n1.E, A=&RigVMFunction_MathDoubleMul(A=@n1.A, B=@n1.D).Result).Result)
  ]}
 Multiply_1_ContainedGraph {nodes [
   n1 = RigVMFunctionEntryNode()
   n2 = &RigVMFunction_MathDoubleMul(A=@n1.A, B=@n1.D)
   RigVMFunctionReturnNode(Result=&RigVMFunction_MathDoubleMul(B=@n1.F, A=&RigVMFunction_MathDoubleMul(B=@n1.E, A=@n2.Result).Result).Result)
  ]}
 Multiply_1_ContainedGraph {nodes [
   n1 = RigVMFunctionEntryNode()
   n2 = &RigVMFunction_MathDoubleMul(B=@n1.E, A=&RigVMFunction_MathDoubleMul(A=@n1.A, B=@n1.D).Result)
   RigVMFunctionReturnNode(Result=&RigVMFunction_MathDoubleMul(B=@n1.G, A=&RigVMFunction_MathDoubleMul(B=@n1.F, A=@n2.Result).Result).Result)
  ]}
 Multiply_1_ContainedGraph {nodes [
   n1 = RigVMFunctionEntryNode()
   n2 = &RigVMFunction_MathDoubleMul(B=@n1.E, A=&RigVMFunction_MathDoubleMul(A=@n1.A, B=@n1.D).Result)
   RigVMFunctionReturnNode(Result=&RigVMFunction_MathDoubleMul(B=@n1.G, A=&RigVMFunction_MathDoubleMul(B=@n1.F, A=@n2.Result).Result).Result)
  ]}
 New Function_ContainedGraph {nodes [
   RigVMFunctionEntryNode()
   > Constraints = RigVMFunctionReferenceNode()
   > HeadAdjustments = RigVMFunctionReferenceNode()
   > RigVMFunction_Sequence = RigVMAggregateNode()
   DeltaRotation_1 = RigVMFunctionReferenceNode(
    child=spine_02
    max=80
    parent=spine_02_null
    parentType=Null
    primAxis=2
    rotationOrder=ZYX
    twistAxis=(X=1,Y=0,Z=0)
    ExecuteContext=@RigVMFunction_Sequence.B
   )
   RigVMFunction_MathBoolToFloat = &RigVMFunction_MathBoolToFloat(Value=RigVMVariableNode() {variable=isWearingBackBling}.Value)
   n1 = &RigVMFunction_MathQuaternionFromEuler(Euler=(X=0,Y=10,Z=0))
   n2 = /Script/RigVM.RigVMFunction_MathVectorMake(X=0, Y=0, Z=0) {is=struct RigVMUnitNode}
   > &RigUnit_ModifyTransforms(
    ItemToModify="((Item=(Type=Bone,Name=\"dyn_hair_M_04\"),Transform=(Rotation=(X=0,Y=0,Z=0,W=1),Translation=(X=0,Y=0,Z=0),Scale3D=(X=1,Y=1,Z=1))),(Item=(Type=Bone,Name=\"dyn_hair_L_04\"),Transform=(Rotation=(X=0,Y=0,Z=0,W=1),Translation=(X=0,Y=0,Z=0),Scale3D=(X=1,Y=1,Z=1))),( ...(truncated)"
    WeightMinimum=0
    Weight=@RigVMFunction_MathBoolToFloat.Result
    ItemToModify.0.Transform.Rotation=@n1.Result
    ItemToModify.1.Transform.Rotation=@n1.Result
    ItemToModify.2.Transform.Rotation=@n1.Result
    ItemToModify.0.Transform.Translation=@n2.Result
    ItemToModify.1.Transform.Translation=@n2.Result
    ItemToModify.2.Transform.Translation=@n2.Result
   )
   RigVMFunctionReturnNode(ExecuteContext=@RigVMFunction_Sequence.D)
   n3 = &RigUnit_ItemArray(Items=((Type=Bone,Name="base_hair_M"),(Type=Bone,Name="dyn_hair_M_01"),(Type=Bone,Name="dyn_hair_M_02"),(Type=Bone,Name="dyn_hair_M_03"),(Type=Bone,Name="dyn_hair_M_04"),(Type=Bone,Name="dyn_hair_M_05"),(Type=Bone,Name="dyn_hair_M_06")))
   SplineIK_2 = RigVMFunctionReferenceNode(
    Primary Axis=(X=1,Y=0,Z=0)
    Samples Per Segment=8
    Secondary Spline Direction=(X=0,Y=0,Z=1)
    Spline Mode=BSpline
    Stretch=true
    Up Axis=(X=0,Y=0,Z=1)
    debug=false
    Bones=@n3.Items
    Controls=RigVMVariableNode() {variable=dyn_hair_m_controls}.Value
    ExecuteContext=@RigVMFunction_Sequence.A
   )
   RigUnit_ItemArray = &RigUnit_ItemArray(Items=((Type=Bone,Name="base_hair_L"),(Type=Bone,Name="dyn_hair_L_01"),(Type=Bone,Name="dyn_hair_L_02"),(Type=Bone,Name="dyn_hair_L_03"),(Type=Bone,Name="dyn_hair_L_04"),(Type=Bone,Name="dyn_hair_L_05"),(Type=Bone,Name="dyn_hair_L_06")))
   > SplineIK_2_1 = RigVMFunctionReferenceNode(
    Primary Axis=(X=1,Y=0,Z=0)
    Samples Per Segment=8
    Secondary Spline Direction=(X=0,Y=0,Z=1)
    Spline Mode=BSpline
    Stretch=true
    Up Axis=(X=0,Y=0,Z=1)
    debug=false
    Bones=@RigUnit_ItemArray.Items
    Controls=RigVMVariableNode() {variable=dyn_hair_l_controls}.Value
   )
   n4 = &RigUnit_ItemArray(Items=((Type=Bone,Name="base_hair_R"),(Type=Bone,Name="dyn_hair_R_01"),(Type=Bone,Name="dyn_hair_R_02"),(Type=Bone,Name="dyn_hair_R_03"),(Type=Bone,Name="dyn_hair_R_04"),(Type=Bone,Name="dyn_hair_R_05"),(Type=Bone,Name="dyn_hair_R_06")))
   > SplineIK_2_1_1 = RigVMFunctionReferenceNode(
    Primary Axis=(X=1,Y=0,Z=0)
    Samples Per Segment=8
    Secondary Spline Direction=(X=0,Y=0,Z=1)
    Spline Mode=BSpline
    Stretch=true
    Up Axis=(X=0,Y=0,Z=1)
    debug=false
    Bones=@n4.Items
    Controls=RigVMVariableNode() {variable=dyn_hair_r_controls}.Value
   )
  ]}
 New Function_ContainedGraph {nodes [
   n1 = RigVMFunctionEntryNode()
   RigVMFunction_ControlFlowBranch = &RigVMFunction_ControlFlowBranch(Condition=@n1.Debug, ExecuteContext=@n1.ExecuteContext)
   RigVMFunctionReturnNode(ExecuteContext=@RigVMFunction_ControlFlowBranch.Completed)
   RigUnit_ItemArray = &RigUnit_ItemArray(Items="((Type=Bone,Name=\"spine_01\"),(Type=Bone,Name=\"spine_02\"),(Type=Bone,Name=\"spine_03\"),(Type=Bone,Name=\"spine_04\"),(Type=Bone,Name=\"spine_05\"),(Type=Bone,Name=\"neck_01\"),(Type=Bone,Name=\"neck_02\"),(Type=Bone,Name=\"head\"),(Type=Bone,Name=\"clavicle_l\"),(Type=B ...(truncated)")
   DISPATCH_RigVMDispatch_ArrayIterator = RigVMDispatchNode(Array=@RigUnit_ItemArray.Items, ExecuteContext=@RigVMFunction_ControlFlowBranch.True) {template="DISPATCH_RigVMDispatch_ArrayIterator(in Array,out Element,out Index,out Count,out Ratio)"}
   Concat = /Script/RigVM.RigVMFunction_NameConcat(B=_ctrl, A=@DISPATCH_RigVMDispatch_ArrayIterator.Element.Name) {is=struct RigVMUnitNode, template="Concat::Execute(in A,in B,out Result)"}
   > &RigUnit_SetTransform(
    Item=(Type=Bone,Name="spine_01")
    Weight=1
    bInitial=False
    Item.Name=@DISPATCH_RigVMDispatch_ArrayIterator.Element.Name
    Value=&RigUnit_GetTransform(Item=(Type=Control,Name="spine_01_ctrl"), bInitial=False, Item.Name=@Concat.Result).Transform
   )
   &RigUnit_SetTransform(
    Item=(Type=Bone,Name="pelvis")
    Weight=1
    bInitial=False
    bPropagateToChildren=false
    Value=&RigUnit_GetTransform(Item=(Type=Control,Name="pelvis_ctrl"), bInitial=False).Transform
    ExecutePin=@DISPATCH_RigVMDispatch_ArrayIterator.Completed
   )
  ]}
 New Function_ContainedGraph {nodes [
   RigVMFunctionEntryNode()
   > &RigUnit_ParentConstraint(
    AdvancedSettings=(RotationOrderForFilter=XZY)
    Child=(Type=Control,Name="base_hair_R_ctrl")
    Filter=(TranslationFilter=(bX=True,bY=True,bZ=True),RotationFilter=(bX=True,bY=True,bZ=True),ScaleFilter=(bX=False,bY=False,bZ=False))
    Parents=((Item=(Type=Bone,Name="head"),Weight=1))
    Weight=1
    bMaintainOffset=true
   )
   > &RigUnit_ParentConstraint(
    AdvancedSettings=(RotationOrderForFilter=XZY)
    Child=(Type=Control,Name="base_hair_M_ctrl")
    Filter=(TranslationFilter=(bX=True,bY=True,bZ=True),RotationFilter=(bX=True,bY=True,bZ=True),ScaleFilter=(bX=False,bY=False,bZ=False))
    Parents=((Item=(Type=Bone,Name="head"),Weight=1))
    Weight=1
   )
   > &RigUnit_ParentConstraint(
    AdvancedSettings=(RotationOrderForFilter=XZY)
    Child=(Type=Control,Name="base_hair_L_ctrl")
    Filter=(TranslationFilter=(bX=True,bY=True,bZ=True),RotationFilter=(bX=True,bY=True,bZ=True),ScaleFilter=(bX=False,bY=False,bZ=False))
    Parents=((Item=(Type=Bone,Name="head"),Weight=1))
    Weight=1
   )
   > &RigUnit_ParentConstraint(
    AdvancedSettings=(RotationOrderForFilter=XZY)
    Child=(Type=Control,Name="dyn_hair_R_01_ctrl")
    Filter=(TranslationFilter=(bX=True,bY=True,bZ=True),RotationFilter=(bX=True,bY=True,bZ=True),ScaleFilter=(bX=False,bY=False,bZ=False))
    Parents=((Item=(Type=Bone,Name="head"),Weight=0.8),(Item=(Type=Bone,Name="spine_05"),Weight=0.2))
    Weight=1
   )
   > &RigUnit_ParentConstraint(
    AdvancedSettings=(RotationOrderForFilter=XZY)
    Child=(Type=Control,Name="dyn_hair_M_01_ctrl")
    Filter=(TranslationFilter=(bX=True,bY=True,bZ=True),RotationFilter=(bX=True,bY=True,bZ=True),ScaleFilter=(bX=False,bY=False,bZ=False))
    Parents=((Item=(Type=Bone,Name="head"),Weight=0.8),(Item=(Type=Bone,Name="spine_05"),Weight=0.2))
    Weight=1
   )
   > &RigUnit_ParentConstraint(
    AdvancedSettings=(RotationOrderForFilter=XZY)
    Child=(Type=Control,Name="dyn_hair_L_01_ctrl")
    Filter=(TranslationFilter=(bX=True,bY=True,bZ=True),RotationFilter=(bX=True,bY=True,bZ=True),ScaleFilter=(bX=False,bY=False,bZ=False))
    Parents=((Item=(Type=Bone,Name="head"),Weight=0.8),(Item=(Type=Bone,Name="spine_05"),Weight=0.2))
    Weight=1
   )
   > &RigUnit_ParentConstraint(
    AdvancedSettings=(RotationOrderForFilter=XZY)
    Child=(Type=Control,Name="dyn_hair_R_02_ctrl")
    Filter=(TranslationFilter=(bX=true,bY=true,bZ=true),RotationFilter=(bX=true,bY=true,bZ=True),ScaleFilter=(bX=False,bY=False,bZ=False))
    Parents=((Item=(Type=Bone,Name="head"),Weight=0.5),(Item=(Type=Bone,Name="spine_05"),Weight=0.5))
    Weight=1
    bMaintainOffset=true
   )
   > &RigUnit_ParentConstraint(
    AdvancedSettings=(RotationOrderForFilter=XZY)
    Child=(Type=Control,Name="dyn_hair_M_02_ctrl")
    Filter=(TranslationFilter=(bX=True,bY=True,bZ=True),RotationFilter=(bX=True,bY=True,bZ=True),ScaleFilter=(bX=False,bY=False,bZ=False))
    Parents=((Item=(Type=Bone,Name="head"),Weight=0.5),(Item=(Type=Bone,Name="spine_05"),Weight=0.5))
    Weight=1
    bMaintainOffset=true
   )
   > &RigUnit_ParentConstraint(
    AdvancedSettings=(RotationOrderForFilter=XZY)
    Child=(Type=Control,Name="dyn_hair_L_02_ctrl")
    Filter=(TranslationFilter=(bX=True,bY=True,bZ=True),RotationFilter=(bX=True,bY=True,bZ=True),ScaleFilter=(bX=False,bY=False,bZ=False))
    Parents=((Item=(Type=Bone,Name="head"),Weight=0.5),(Item=(Type=Bone,Name="spine_05"),Weight=0.5))
    Weight=1
    bMaintainOffset=true
   )
   > &RigUnit_ParentConstraint(
    AdvancedSettings=(RotationOrderForFilter=XZY)
    Child=(Type=Control,Name="dyn_hair_R_03_ctrl")
    Filter=(TranslationFilter=(bX=True,bY=True,bZ=True),RotationFilter=(bX=True,bY=True,bZ=True),ScaleFilter=(bX=False,bY=False,bZ=False))
    Parents=((Item=(Type=Bone,Name="upperarm_twist_01_r"),Weight=0.05),(Item=(Type=Bone,Name="spine_05"),Weight=0.2),(Item=(Type=Bone,Name="clavicle_r"),Weight=0.45),(Item=(Type=Bone,Name="head"),Weight=0.3))
    Weight=1
   )
   > &RigUnit_ParentConstraint(
    AdvancedSettings=(RotationOrderForFilter=XZY)
    Child=(Type=Control,Name="dyn_hair_M_03_ctrl")
    Filter=(TranslationFilter=(bX=True,bY=True,bZ=True),RotationFilter=(bX=True,bY=True,bZ=True),ScaleFilter=(bX=False,bY=False,bZ=False))
    Parents=((Item=(Type=Bone,Name="clavicle_l"),Weight=0.2),(Item=(Type=Bone,Name="clavicle_r"),Weight=0.2),(Item=(Type=Bone,Name="spine_05"),Weight=0.3),(Item=(Type=Bone,Name="head"),Weight=0.3))
    Weight=1
   )
   > &RigUnit_ParentConstraint(
    AdvancedSettings=(RotationOrderForFilter=XZY)
    Child=(Type=Control,Name="dyn_hair_L_03_ctrl")
    Filter=(TranslationFilter=(bX=True,bY=True,bZ=True),RotationFilter=(bX=True,bY=True,bZ=True),ScaleFilter=(bX=False,bY=False,bZ=False))
    Parents=((Item=(Type=Bone,Name="upperarm_twist_01_l"),Weight=0.05),(Item=(Type=Bone,Name="spine_05"),Weight=0.2),(Item=(Type=Bone,Name="clavicle_l"),Weight=0.45),(Item=(Type=Bone,Name="head"),Weight=0.3))
    Weight=1
   )
   > &RigUnit_ParentConstraint(
    AdvancedSettings=(RotationOrderForFilter=XZY)
    Child=(Type=Control,Name="dyn_hair_R_04_ctrl")
    Filter=(TranslationFilter=(bX=True,bY=True,bZ=True),RotationFilter=(bX=True,bY=True,bZ=True),ScaleFilter=(bX=False,bY=False,bZ=False))
    Parents="((Item=(Type=Bone,Name=\"spine_02\"),Weight=0.0014),(Item=(Type=Bone,Name=\"spine_03\"),Weight=0.0235),(Item=(Type=Bone,Name=\"spine_04\"),Weight=0.0379),(Item=(Type=Bone,Name=\"upperarm_twist_01_r\"),Weight=0.0733),(Item=(Type=Bone,Name=\"spine_05\"),Weight=0.3523) ...(truncated)"
    Weight=1
   )
   > &RigUnit_ParentConstraint(
    AdvancedSettings=(RotationOrderForFilter=XZY)
    Child=(Type=Control,Name="dyn_hair_M_04_ctrl")
    Filter=(TranslationFilter=(bX=True,bY=True,bZ=True),RotationFilter=(bX=True,bY=True,bZ=True),ScaleFilter=(bX=False,bY=False,bZ=False))
    Parents="((Item=(Type=Bone,Name=\"spine_01\"),Weight=0.0001),(Item=(Type=Bone,Name=\"spine_02\"),Weight=0.0068),(Item=(Type=Bone,Name=\"clavicle_l\"),Weight=0.063),(Item=(Type=Bone,Name=\"clavicle_r\"),Weight=0.0705),(Item=(Type=Bone,Name=\"spine_03\"),Weight=0.1021),(Item=( ...(truncated)"
    Weight=1
   )
   > &RigUnit_ParentConstraint(
    AdvancedSettings=(RotationOrderForFilter=XZY)
    Child=(Type=Control,Name="dyn_hair_L_04_ctrl")
    Filter=(TranslationFilter=(bX=True,bY=True,bZ=True),RotationFilter=(bX=True,bY=True,bZ=True),ScaleFilter=(bX=False,bY=False,bZ=False))
    Parents="((Item=(Type=Bone,Name=\"spine_02\"),Weight=0.0014),(Item=(Type=Bone,Name=\"spine_03\"),Weight=0.0235),(Item=(Type=Bone,Name=\"spine_04\"),Weight=0.038),(Item=(Type=Bone,Name=\"upperarm_twist_01_l\"),Weight=0.0733),(Item=(Type=Bone,Name=\"spine_05\"),Weight=0.3528), ...(truncated)"
    Weight=1
   )
   > &RigUnit_ParentConstraint(
    AdvancedSettings=(RotationOrderForFilter=XZY)
    Child=(Type=Control,Name="dyn_hair_R_05_ctrl")
    Filter=(TranslationFilter=(bX=True,bY=True,bZ=True),RotationFilter=(bX=True,bY=True,bZ=True),ScaleFilter=(bX=False,bY=False,bZ=False))
    Parents=((Item=(Type=Bone,Name="clavicle_r"),Weight=0.1),(Item=(Type=Bone,Name="spine_04"),Weight=0.15),(Item=(Type=Bone,Name="spine_05"),Weight=0.15),(Item=(Type=Bone,Name="spine_02"),Weight=0.2),(Item=(Type=Bone,Name="spine_03"),Weight=0.4))
    Weight=1
   )
   > &RigUnit_ParentConstraint(
    AdvancedSettings=(RotationOrderForFilter=XZY)
    Child=(Type=Control,Name="dyn_hair_M_05_ctrl")
    Filter=(TranslationFilter=(bX=True,bY=True,bZ=True),RotationFilter=(bX=True,bY=True,bZ=True),ScaleFilter=(bX=False,bY=False,bZ=False))
    Parents=((Item=(Type=Bone,Name="spine_04"),Weight=0.15),(Item=(Type=Bone,Name="spine_05"),Weight=0.1),(Item=(Type=Bone,Name="spine_02"),Weight=0.25),(Item=(Type=Bone,Name="spine_03"),Weight=0.5))
    Weight=1
   )
   > &RigUnit_ParentConstraint(
    AdvancedSettings=(RotationOrderForFilter=XZY)
    Child=(Type=Control,Name="dyn_hair_L_05_ctrl")
    Filter=(TranslationFilter=(bX=True,bY=True,bZ=True),RotationFilter=(bX=True,bY=True,bZ=True),ScaleFilter=(bX=False,bY=False,bZ=False))
    Parents=((Item=(Type=Bone,Name="clavicle_l"),Weight=0.1),(Item=(Type=Bone,Name="spine_04"),Weight=0.15),(Item=(Type=Bone,Name="spine_05"),Weight=0.15),(Item=(Type=Bone,Name="spine_02"),Weight=0.2),(Item=(Type=Bone,Name="spine_03"),Weight=0.4))
    Weight=1
   )
   > &RigUnit_ParentConstraint(
    AdvancedSettings=(RotationOrderForFilter=XZY)
    Child=(Type=Control,Name="dyn_hair_R_06_ctrl")
    Filter=(TranslationFilter=(bX=True,bY=True,bZ=True),RotationFilter=(bX=True,bY=True,bZ=True),ScaleFilter=(bX=False,bY=False,bZ=False))
    Parents=((Item=(Type=Bone,Name="spine_02"),Weight=0.175),(Item=(Type=Bone,Name="spine_01"),Weight=0.425),(Item=(Type=Bone,Name="pelvis"),Weight=0.4))
    Weight=1
   )
   > &RigUnit_ParentConstraint(
    AdvancedSettings=(RotationOrderForFilter=XZY)
    Child=(Type=Control,Name="dyn_hair_M_06_ctrl")
    Filter=(TranslationFilter=(bX=True,bY=True,bZ=True),RotationFilter=(bX=True,bY=True,bZ=True),ScaleFilter=(bX=False,bY=False,bZ=False))
    Parents=((Item=(Type=Bone,Name="spine_02"),Weight=0.175),(Item=(Type=Bone,Name="spine_01"),Weight=0.425),(Item=(Type=Bone,Name="pelvis"),Weight=0.4))
    Weight=1
   )
   > &RigUnit_ParentConstraint(
    AdvancedSettings=(RotationOrderForFilter=XZY)
    Child=(Type=Control,Name="dyn_hair_L_06_ctrl")
    Filter=(TranslationFilter=(bX=True,bY=True,bZ=True),RotationFilter=(bX=True,bY=True,bZ=True),ScaleFilter=(bX=False,bY=False,bZ=False))
    Parents=((Item=(Type=Bone,Name="spine_02"),Weight=0.175),(Item=(Type=Bone,Name="spine_01"),Weight=0.425),(Item=(Type=Bone,Name="pelvis"),Weight=0.4))
    Weight=1
   )
   > RigVMFunctionReturnNode()
  ]}
 New Function_ContainedGraph {nodes [
   n1 = RigVMFunctionEntryNode()
   RigVMFunction_MathBoolToInteger = &RigVMFunction_MathBoolToInteger(Value=@n1.onlyTwist)
   RigVMFunction_MathQuaternionSwingTwist = &RigVMFunction_MathQuaternionSwingTwist(
    TwistAxis=@n1.twistAxis
    Input=&RigUnit_GetRelativeTransformForItem(
     Child=(Type=Bone,Name="None")
     Parent=(Type=Bone,Name="None")
     bChildInitial=False
     bParentInitial=False
     Child.Name=@n1.child
     Parent.Name=@n1.parent
     Parent.Type=@n1.parentType
    ).RelativeTransform.Rotation
   )
   RigVMFunction_MathQuaternionToEuler = &RigVMFunction_MathQuaternionToEuler(
    Value=RigVMDispatchNode(
     Values=((),())
     Index=@RigVMFunction_MathBoolToInteger.Result
     Values.0=@RigVMFunction_MathQuaternionSwingTwist.Swing
     Values.1=@RigVMFunction_MathQuaternionSwingTwist.Twist
    ) {template="DISPATCH_RigVMDispatch_SelectInt32(in Index,in Values,out Result)"}.Result
    RotationOrder=@n1.rotationOrder
   )
   > DISPATCH_RigVMDispatch_Print = RigVMDispatchNode(ScreenColor=(R=1,G=1,B=1,A=1), ScreenDuration=1.5, Enabled=@n1.debugRotation, Value=@RigVMFunction_MathQuaternionToEuler.Result) {template="DISPATCH_RigVMDispatch_Print(in Prefix,in Value,in Enabled,in ScreenDuration,in ScreenColor)"}
   Clamp = &RigVMFunction_MathIntClamp(Value=@n1.primAxis)
   DISPATCH_RigVMDispatch_SelectInt32 = RigVMDispatchNode(
    Values=(0,0,0)
    Index=@Clamp.Result
    Values.0=@RigVMFunction_MathQuaternionToEuler.Result.X
    Values.1=@RigVMFunction_MathQuaternionToEuler.Result.Y
    Values.2=@RigVMFunction_MathQuaternionToEuler.Result.Z
   ) {template="DISPATCH_RigVMDispatch_SelectInt32(in Index,in Values,out Result)"}
   Remap = &RigVMFunction_MathDoubleRemap(
    TargetMaximum=1
    TargetMinimum=0
    bClamp=False
    Value=@DISPATCH_RigVMDispatch_SelectInt32.Result
    SourceMaximum=@n1.max
    SourceMinimum=@n1.min
   )
   > RigVMFunctionReturnNode(outputWeight=&RigVMFunction_MathDoubleClamp(Value=@Remap.Result).Result)
  ]}
 New Function_ContainedGraph {nodes [
   n1 = RigVMFunctionEntryNode()
   > DISPATCH_RigVMDispatch_ArrayIterator = RigVMDispatchNode(Array=@n1.transforms) {template="DISPATCH_RigVMDispatch_ArrayIterator(in Array,out Element,out Index,out Count,out Ratio)"}
   Multiply = &RigVMFunction_MathTransformMul(
    A=@DISPATCH_RigVMDispatch_ArrayIterator.Element.Transform
    B=&RigUnit_GetTransform(
     Item=(Type=Bone,Name="dyn_L_hair_3")
     Space=LocalSpace
     bInitial=False
     Item.Name=@DISPATCH_RigVMDispatch_ArrayIterator.Element.Bone
     Item.Type=@n1.type
    ).Transform
   ) {template="Multiply::Execute(in A,in B,out Result)"}
   > &RigUnit_SetTransform(
    Item=(Type=Bone,Name="dyn_L_hair_3")
    Space=LocalSpace
    bInitial=false
    Value=@Multiply.Result
    bPropagateToChildren=@n1.inheristTransform
    Weight=@n1.weight
    Item.Name=@DISPATCH_RigVMDispatch_ArrayIterator.Element.Bone
    Item.Type=@n1.type
   )
   RigVMFunctionReturnNode(ExecuteContext=@DISPATCH_RigVMDispatch_ArrayIterator.Completed)
  ]}
 New Function_ContainedGraph {nodes [
   RigVMFunctionEntryNode()
   > n1 = &RigUnit_SphericalPoseReader(
    ActiveRegionScaleFactors=(PositiveWidth=0.5,NegativeWidth=0.5,PositiveHeight=1,NegativeHeight=0.5)
    ActiveRegionSize=0.18
    Debug=(bDrawDebug=false,bDraw2D=False,bDrawLocalAxes=False,DebugScale=25,DebugSegments=20,DebugThickness=0.25)
    DriverAxis=(X=1,Y=0,Z=0)
    DriverItem=(Type=Bone,Name="clavicle_l")
    FalloffRegionScaleFactors=(PositiveWidth=0.5,NegativeWidth=0.5,PositiveHeight=0.5,NegativeHeight=0.5)
    FalloffSize=0.4
    FlipHeightScaling=False
    FlipWidthScaling=False
    OptionalParentItem=(Type=Bone,Name="None")
    RotationOffset=(X=0,Y=90,Z=130)
   )
   > AlterTransform = RigVMFunctionReferenceNode(
    type=Bone
    inheristTransform=False
    transforms="((Bone=\"dyn_hair_L_03\",Transform=(Rotation=(X=0,Y=0,Z=0,W=1),Translation=(X=0,Y=1,Z=3),Scale3D=(X=1,Y=1,Z=1))),(Bone=\"dyn_hair_L_04\",Transform=(Rotation=(X=0,Y=0,Z=0,W=1),Translation=(X=0,Y=2,Z=1),Scale3D=(X=1,Y=1,Z=1))),(Bone=\"dyn_hair_L_02\",Transform=(Ro ...(truncated)"
    weight=@n1.OutputParam
   )
   > n2 = &RigUnit_SphericalPoseReader(
    ActiveRegionScaleFactors=(PositiveWidth=0.5,NegativeWidth=0.5,PositiveHeight=0.5,NegativeHeight=1)
    ActiveRegionSize=0.18
    Debug=(bDrawDebug=false,bDraw2D=False,bDrawLocalAxes=False,DebugScale=25,DebugSegments=20,DebugThickness=0.25)
    DriverAxis=(X=-1,Y=0,Z=0)
    DriverItem=(Type=Bone,Name="clavicle_r")
    FalloffRegionScaleFactors=(PositiveWidth=0.5,NegativeWidth=0.5,PositiveHeight=0.5,NegativeHeight=0.5)
    FalloffSize=0.4
    FlipHeightScaling=False
    FlipWidthScaling=False
    OptionalParentItem=(Type=Bone,Name="None")
    RotationOffset=(X=0,Y=-90,Z=130)
   )
   > AlterTransform_1 = RigVMFunctionReferenceNode(
    type=Bone
    inheristTransform=False
    transforms="((Bone=\"dyn_hair_R_03\",Transform=(Rotation=(X=0,Y=0,Z=0,W=1),Translation=(X=0,Y=-1,Z=5),Scale3D=(X=1,Y=1,Z=1))),(Bone=\"dyn_hair_R_04\",Transform=(Rotation=(X=0,Y=0,Z=0,W=1),Translation=(X=0,Y=-2,Z=1),Scale3D=(X=1,Y=1,Z=1))),(Bone=\"dyn_hair_R_02\",Transform=( ...(truncated)"
    weight=@n2.OutputParam
   )
   > RigVMFunctionReturnNode()
  ]}
 New Function_ContainedGraph {nodes [
   RigVMFunctionEntryNode()
   > RigVMFunction_Sequence = RigVMAggregateNode()
   n1 = &RigVMFunction_Sequence(ExecuteContext=@RigVMFunction_Sequence.C)
   DeltaRotation_8 = RigVMFunctionReferenceNode(
    child=spine_05
    max=-50
    min=-10
    onlyTwist=true
    parent=spine_01
    parentType=Bone
    primAxis=0
    rotationOrder=ZYX
    twistAxis=(X=1,Y=0,Z=0)
    ExecuteContext=@n1.A
   )
   n2 = &RigVMFunction_AnimEvalRichCurve(
    Curve=(EditorCurveData=(Keys=((InterpMode=RCIM_Cubic),(InterpMode=RCIM_Cubic,TangentMode=RCTM_User,Time=0.5279,Value=0.4056,ArriveTangent=1.3758,LeaveTangent=1.3758),(InterpMode=RCIM_Cubic,Time=1,Value=1))))
    SourceMinimum=0
    TargetMinimum=0
    Value=@DeltaRotation_8.outputWeight
   )
   > AlterTransform_16 = RigVMFunctionReferenceNode(
    type=Bone
    inheristTransform=False
    transforms="((Bone=\"dyn_hair_L_03\",Transform=(Rotation=(X=0,Y=0,Z=0,W=1),Translation=(X=0,Y=1,Z=3),Scale3D=(X=1,Y=1,Z=1))),(Bone=\"dyn_hair_L_04\",Transform=(Rotation=(X=0,Y=0,Z=0,W=1),Translation=(X=0,Y=3,Z=2),Scale3D=(X=1,Y=1,Z=1))),(Bone=\"dyn_hair_L_02\",Transform=(Ro ...(truncated)"
    weight=@n2.Result
   )
   > DeltaRotation_9 = RigVMFunctionReferenceNode(child=spine_05, max=50, min=10, onlyTwist=true, parent=spine_01, parentType=Bone, primAxis=0, rotationOrder=ZYX, twistAxis=(X=1,Y=0,Z=0))
   RigVMFunction_AnimEvalRichCurve = &RigVMFunction_AnimEvalRichCurve(
    Curve=(EditorCurveData=(Keys=((InterpMode=RCIM_Cubic),(InterpMode=RCIM_Cubic,TangentMode=RCTM_User,Time=0.5279,Value=0.4056,ArriveTangent=1.3758,LeaveTangent=1.3758),(InterpMode=RCIM_Cubic,Time=1,Value=1))))
    SourceMinimum=0
    TargetMinimum=0
    Value=@DeltaRotation_9.outputWeight
   )
   > AlterTransform_17 = RigVMFunctionReferenceNode(
    type=Bone
    inheristTransform=False
    transforms="((Bone=\"dyn_hair_R_03\",Transform=(Rotation=(X=0,Y=0,Z=0,W=1),Translation=(X=0,Y=-1,Z=0),Scale3D=(X=1,Y=1,Z=1))),(Bone=\"dyn_hair_R_04\",Transform=(Rotation=(X=0,Y=0,Z=0,W=1),Translation=(X=0,Y=-3,Z=0),Scale3D=(X=1,Y=1,Z=1))),(Bone=\"dyn_hair_R_02\",Transform=( ...(truncated)"
    weight=@RigVMFunction_AnimEvalRichCurve.Result
   )
   n3 = &RigVMFunction_ControlFlowBranch(Condition=RigVMVariableNode() {variable=pub isTargeting}.Value, ExecuteContext=@n1.B)
   n4 = RigVMVariableNode() {variable=pub aimPitch}
   n5 = &RigVMFunction_AnimEvalRichCurve(
    Curve=(EditorCurveData=(Keys=((InterpMode=RCIM_Cubic),(InterpMode=RCIM_Cubic,TangentMode=RCTM_User,Time=0.4302,Value=0.2017,ArriveTangent=0.8571,LeaveTangent=0.8571),(Time=1,Value=1))))
    SourceMinimum=0
    TargetMinimum=0
    Value=&RigVMFunction_MathDoubleMul(
     B=&RigVMFunction_MathDoubleRemap(SourceMaximum=0, SourceMinimum=88, TargetMaximum=0, TargetMinimum=1, Value=@n4.Value).Result
     A=RigVMVariableNode() {variable=pub isAssualtedRifleEquipped}.Value
    ).Result
   )
   &RigUnit_ModifyTransforms(
    ItemToModify=((Item=(Type=Bone,Name="base_hair_R"),Transform=(Rotation=(X=0,Y=0,Z=0,W=1),Translation=(X=0,Y=3,Z=1.5),Scale3D=(X=1,Y=1,Z=1))))
    WeightMinimum=0
    Weight=@n5.Result
    ExecutePin=@n3.True
   )
   > AlterTransform_18 = RigVMFunctionReferenceNode(
    type=Bone
    inheristTransform=False
    transforms=((Bone="dyn_hair_R_01",Transform=(Rotation=(X=-0.0872,Y=0,Z=0,W=0.9962),Translation=(X=0,Y=0,Z=0),Scale3D=(X=1,Y=1,Z=1))))
    weight=@n5.Result
   )
   RigVMFunctionReturnNode(ExecuteContext=@RigVMFunction_Sequence.D)
   n6 = &RigVMFunction_ControlFlowBranch(Condition=RigVMVariableNode() {variable=pub diveFloating}.Value, ExecuteContext=@RigVMFunction_Sequence.A)
   claviclesBackForward = RigVMFunctionReferenceNode(ExecuteContext=@n6.True)
   RigVMFunction_ControlFlowBranch = &RigVMFunction_ControlFlowBranch(Condition=RigVMVariableNode() {variable=pub disableUpperLimbsCorrection}.Value, ExecuteContext=@RigVMFunction_Sequence.B)
   RigVMFunction_Sequence_1 = RigVMAggregateNode(ExecuteContext=@RigVMFunction_ControlFlowBranch.False)
   n7 = &RigUnit_SphericalPoseReader(
    ActiveRegionScaleFactors=(PositiveWidth=1,NegativeWidth=1,PositiveHeight=1,NegativeHeight=1)
    ActiveRegionSize=0.1
    Debug=(bDrawDebug=false,bDraw2D=false,bDrawLocalAxes=False,DebugScale=25,DebugSegments=20,DebugThickness=0.25)
    DriverAxis=(X=1,Y=0,Z=0)
    DriverItem=(Type=Bone,Name="upperarm_l")
    FalloffRegionScaleFactors=(PositiveWidth=1,NegativeWidth=1,PositiveHeight=1,NegativeHeight=1)
    FalloffSize=0.3
    FlipHeightScaling=False
    FlipWidthScaling=False
    OptionalParentItem=(Type=Bone,Name="None")
    RotationOffset=(X=-140,Y=50,Z=0)
    ExecutePin=@RigVMFunction_Sequence_1.A
   )
   > AlterTransform = RigVMFunctionReferenceNode(
    type=Bone
    inheristTransform=False
    transforms="((Bone=\"dyn_hair_L_03\",Transform=(Rotation=(X=0,Y=0,Z=0,W=1),Translation=(X=0,Y=-4,Z=0.5),Scale3D=(X=1,Y=1,Z=1))),(Bone=\"dyn_hair_L_04\",Transform=(Rotation=(X=0,Y=0,Z=0,W=1),Translation=(X=0,Y=-4,Z=0),Scale3D=(X=1,Y=1,Z=1))),(Bone=\"dyn_hair_L_02\",Transform ...(truncated)"
    weight=@n7.OutputParam
   )
   > n8 = &RigUnit_SphericalPoseReader(
    ActiveRegionScaleFactors=(PositiveWidth=1,NegativeWidth=1,PositiveHeight=1,NegativeHeight=1)
    ActiveRegionSize=0.1
    Debug=(bDrawDebug=false,DebugScale=25,DebugSegments=20,DebugThickness=0.25)
    DriverAxis=(X=-1,Y=0,Z=0)
    DriverItem=(Type=Bone,Name="upperarm_r")
    FalloffRegionScaleFactors=(PositiveWidth=1,NegativeWidth=1,PositiveHeight=1,NegativeHeight=1)
    FalloffSize=0.3
    FlipHeightScaling=false
    FlipWidthScaling=False
    OptionalParentItem=(Type=Bone,Name="None")
    RotationOffset=(X=40,Y=50,Z=0)
   )
   > AlterTransform_1 = RigVMFunctionReferenceNode(
    type=Bone
    inheristTransform=False
    transforms="((Bone=\"dyn_hair_R_03\",Transform=(Rotation=(X=0,Y=0,Z=0,W=1),Translation=(X=0,Y=4,Z=0.5),Scale3D=(X=1,Y=1,Z=1))),(Bone=\"dyn_hair_R_04\",Transform=(Rotation=(X=0,Y=0,Z=0,W=1),Translation=(X=0,Y=4,Z=0),Scale3D=(X=1,Y=1,Z=1))),(Bone=\"dyn_hair_R_02\",Transform=( ...(truncated)"
    weight=@n8.OutputParam
   )
   claviclesBackForward_1 = RigVMFunctionReferenceNode(ExecuteContext=@RigVMFunction_Sequence_1.B)
   > n9 = &RigUnit_SphericalPoseReader(
    ActiveRegionScaleFactors=(PositiveWidth=1,NegativeWidth=1,PositiveHeight=1,NegativeHeight=1)
    ActiveRegionSize=0.1
    Debug=(bDrawDebug=false,bDraw2D=False,bDrawLocalAxes=False,DebugScale=25,DebugSegments=20,DebugThickness=0.25)
    DriverAxis=(X=1,Y=0,Z=0)
    DriverItem=(Type=Bone,Name="clavicle_l")
    FalloffRegionScaleFactors=(PositiveWidth=0.5,NegativeWidth=0.5,PositiveHeight=0.5,NegativeHeight=0.5)
    FalloffSize=0.3
    FlipHeightScaling=False
    FlipWidthScaling=False
    OptionalParentItem=(Type=Bone,Name="None")
    RotationOffset=(X=-45,Y=6,Z=60)
   )
   > AlterTransform_2 = RigVMFunctionReferenceNode(
    type=Bone
    inheristTransform=False
    transforms=((Bone="dyn_hair_L_02",Transform=(Rotation=(X=0,Y=0,Z=0,W=1),Translation=(X=0,Y=-2,Z=0),Scale3D=(X=1,Y=1,Z=1))),(Bone="dyn_hair_L_03",Transform=(Rotation=(X=0,Y=0,Z=0,W=1),Translation=(X=0,Y=-3,Z=0),Scale3D=(X=1,Y=1,Z=1))))
    weight=@n9.OutputParam
   )
   > n10 = &RigUnit_SphericalPoseReader(
    ActiveRegionScaleFactors=(PositiveWidth=1,NegativeWidth=1,PositiveHeight=1,NegativeHeight=1)
    ActiveRegionSize=0.1
    Debug=(bDrawDebug=false,bDraw2D=False,bDrawLocalAxes=False,DebugScale=25,DebugSegments=20,DebugThickness=0.25)
    DriverAxis=(X=-1,Y=0,Z=0)
    DriverItem=(Type=Bone,Name="clavicle_r")
    FalloffRegionScaleFactors=(PositiveWidth=0.5,NegativeWidth=0.5,PositiveHeight=0.5,NegativeHeight=0.5)
    FalloffSize=0.3
    FlipHeightScaling=False
    FlipWidthScaling=False
    OptionalParentItem=(Type=Bone,Name="None")
    RotationOffset=(X=45,Y=174,Z=30)
   )
   > AlterTransform_3 = RigVMFunctionReferenceNode(
    type=Bone
    inheristTransform=False
    transforms=((Bone="dyn_hair_R_02",Transform=(Rotation=(X=0,Y=0,Z=0,W=1),Translation=(X=0,Y=3,Z=0),Scale3D=(X=1,Y=1,Z=1))),(Bone="dyn_hair_R_03",Transform=(Rotation=(X=0,Y=0,Z=0,W=1),Translation=(X=0,Y=2,Z=0),Scale3D=(X=1,Y=1,Z=1))))
    weight=@n10.OutputParam
   )
   > DeltaRotation = RigVMFunctionReferenceNode(child=clavicle_r, max=-15, parent=clavicle_r_null, parentType=Null, primAxis=1, rotationOrder=ZXY, twistAxis=(X=1,Y=0,Z=0))
   > AlterTransform_4 = RigVMFunctionReferenceNode(
    type=Bone
    inheristTransform=False
    transforms=((Bone="dyn_hair_R_03",Transform=(Rotation=(X=0,Y=0,Z=0,W=1),Translation=(X=-2,Y=1,Z=1),Scale3D=(X=1,Y=1,Z=1))),(Bone="dyn_hair_R_02",Transform=(Rotation=(X=0,Y=0,Z=0,W=1),Translation=(X=-1,Y=0,Z=0),Scale3D=(X=1,Y=1,Z=1))))
    weight=@DeltaRotation.outputWeight
   )
   > DeltaRotation_1 = RigVMFunctionReferenceNode(child=clavicle_l, max=-10, parent=clavicle_l_null, parentType=Null, primAxis=1, rotationOrder=ZXY, twistAxis=(X=1,Y=0,Z=0))
   > AlterTransform_5 = RigVMFunctionReferenceNode(
    type=Bone
    inheristTransform=False
    transforms=((Bone="dyn_hair_L_03",Transform=(Rotation=(X=0,Y=0,Z=0,W=1),Translation=(X=-2,Y=-1,Z=1),Scale3D=(X=1,Y=1,Z=1))),(Bone="dyn_hair_L_02",Transform=(Rotation=(X=0,Y=0,Z=0,W=1),Translation=(X=-1,Y=0,Z=1),Scale3D=(X=1,Y=1,Z=1))))
    weight=@DeltaRotation_1.outputWeight
   )
   n11 = &RigUnit_SphericalPoseReader(
    ActiveRegionScaleFactors=(PositiveWidth=2,NegativeWidth=1,PositiveHeight=5,NegativeHeight=1)
    ActiveRegionSize=0.1
    Debug=(bDrawDebug=false,bDraw2D=False,bDrawLocalAxes=False,DebugScale=25,DebugSegments=20,DebugThickness=0.25)
    DriverAxis=(X=1,Y=0,Z=0)
    DriverItem=(Type=Bone,Name="head")
    FalloffRegionScaleFactors=(PositiveWidth=1.4,NegativeWidth=1.4,PositiveHeight=0.1,NegativeHeight=1.4)
    FalloffSize=0.4
    FlipHeightScaling=false
    FlipWidthScaling=true
    OptionalParentItem=(Type=Bone,Name="spine_05")
    RotationOffset=(X=0,Y=-60,Z=-70)
    ExecutePin=@RigVMFunction_Sequence_1.C
   )
   > &RigUnit_ModifyTransforms(
    ItemToModify="((Item=(Type=Bone,Name=\"None\"),Transform=(Rotation=(X=0,Y=0,Z=0,W=1),Translation=(X=0,Y=0,Z=0),Scale3D=(X=1,Y=1,Z=1))),(Item=(Type=Bone,Name=\"base_hair_R\"),Transform=(Rotation=(X=0,Y=0,Z=0,W=1),Translation=(X=-1,Y=0,Z=0),Scale3D=(X=1,Y=1,Z=1))),(Item=(Type ...(truncated)"
    WeightMinimum=0
    Weight=@n11.OutputParam
   )
  ]}
 New Function_ContainedGraph {nodes [
   RigVMFunctionEntryNode()
   > DeltaRotation_2 = RigVMFunctionReferenceNode(child=neck_01, max=-23, parent=neck_01_null, parentType=Null, primAxis=2, rotationOrder=ZYX, twistAxis=(X=1,Y=0,Z=0))
   n1 = &RigVMFunction_AnimEvalRichCurve(
    Curve=(EditorCurveData=(Keys=((InterpMode=RCIM_Cubic),(InterpMode=RCIM_Cubic,Time=0.3879,Value=0.174,ArriveTangent=1,LeaveTangent=1),(InterpMode=RCIM_Cubic,TangentMode=RCTM_User,Time=1,Value=1,ArriveTangent=1.9233,LeaveTangent=1.9233))))
    SourceMinimum=0
    TargetMinimum=0
    Value=@DeltaRotation_2.outputWeight
   )
   > &RigUnit_ModifyTransforms(
    ItemToModify="((Item=(Type=Control,Name=\"base_hair_L_ctrl\"),Transform=(Rotation=(X=0,Y=-0.1736,Z=0,W=0.9848),Translation=(X=0,Y=0,Z=0),Scale3D=(X=1,Y=1,Z=1))),(Item=(Type=Control,Name=\"dyn_hair_L_01_ctrl\"),Transform=(Rotation=(X=0,Y=0,Z=0,W=1),Translation=(X=0,Y=0,Z=0), ...(truncated)"
    WeightMinimum=0
    Weight=@n1.Result
   )
   > &RigUnit_ModifyTransforms(
    ItemToModify="((Item=(Type=Control,Name=\"base_hair_M_ctrl\"),Transform=(Rotation=(X=0,Y=-0.1736,Z=0,W=0.9848),Translation=(X=1,Y=0,Z=1),Scale3D=(X=1,Y=1,Z=1))),(Item=(Type=Control,Name=\"dyn_hair_M_01_ctrl\"),Transform=(Rotation=(X=0,Y=0,Z=0,W=1),Translation=(X=0,Y=0,Z=2), ...(truncated)"
    WeightMinimum=0
    Weight=@n1.Result
   )
   > &RigUnit_ModifyTransforms(
    ItemToModify="((Item=(Type=Control,Name=\"base_hair_R_ctrl\"),Transform=(Rotation=(X=0,Y=-0.0872,Z=0,W=0.9962),Translation=(X=0,Y=0,Z=1),Scale3D=(X=1,Y=1,Z=1))),(Item=(Type=Control,Name=\"dyn_hair_R_01_ctrl\"),Transform=(Rotation=(X=0,Y=0,Z=0,W=1),Translation=(X=0,Y=-2,Z=3) ...(truncated)"
    WeightMinimum=0
    Weight=@n1.Result
   )
   > DeltaRotation_3 = RigVMFunctionReferenceNode(child=head, max=-40, parent=head_null, parentType=Null, primAxis=2, rotationOrder=ZYX, twistAxis=(X=1,Y=0,Z=0))
   n2 = &RigVMFunction_AnimEvalRichCurve(
    Curve=(EditorCurveData=(Keys=((InterpMode=RCIM_Cubic),(InterpMode=RCIM_Cubic,Time=0.3879,Value=0.174,ArriveTangent=1,LeaveTangent=1),(InterpMode=RCIM_Cubic,TangentMode=RCTM_User,Time=1,Value=1,ArriveTangent=1.9233,LeaveTangent=1.9233))))
    SourceMinimum=0
    TargetMinimum=0
    Value=@DeltaRotation_3.outputWeight
   )
   n3 = &RigVMFunction_MathDoubleMul(B=0.5, A=@n2.Result)
   > &RigUnit_ModifyTransforms(
    ItemToModify="((Item=(Type=Control,Name=\"base_hair_L_ctrl\"),Transform=(Rotation=(X=0,Y=0,Z=0,W=1),Translation=(X=0,Y=0,Z=0),Scale3D=(X=1,Y=1,Z=1))),(Item=(Type=Control,Name=\"dyn_hair_L_01_ctrl\"),Transform=(Rotation=(X=0.342,Y=0,Z=0,W=0.9397),Translation=(X=0,Y=0,Z=3),Sc ...(truncated)"
    WeightMinimum=0
    Weight=@n3.Result
   )
   > &RigUnit_ModifyTransforms(
    ItemToModify="((Item=(Type=Control,Name=\"base_hair_M_ctrl\"),Transform=(Rotation=(X=0,Y=0,Z=0,W=1),Translation=(X=0,Y=0,Z=2),Scale3D=(X=1,Y=1,Z=1))),(Item=(Type=Control,Name=\"dyn_hair_M_01_ctrl\"),Transform=(Rotation=(X=0,Y=0,Z=0,W=1),Translation=(X=0,Y=0,Z=4),Scale3D=(X= ...(truncated)"
    WeightMinimum=0
    Weight=@n3.Result
   )
   > &RigUnit_ModifyTransforms(
    ItemToModify="((Item=(Type=Control,Name=\"base_hair_R_ctrl\"),Transform=(Rotation=(X=-0.1736,Y=0,Z=0,W=0.9848),Translation=(X=0,Y=0,Z=4),Scale3D=(X=1,Y=1,Z=1))),(Item=(Type=Control,Name=\"dyn_hair_R_01_ctrl\"),Transform=(Rotation=(X=-0.342,Y=0,Z=0,W=0.9397),Translation=(X=0 ...(truncated)"
    WeightMinimum=0
    Weight=@n3.Result
   )
   > DeltaRotation_4 = RigVMFunctionReferenceNode(child=head, max=60, parent=head_null, parentType=Null, primAxis=2, rotationOrder=ZYX, twistAxis=(X=1,Y=0,Z=0))
   > &RigUnit_ModifyTransforms(
    ItemToModify="((Item=(Type=Control,Name=\"dyn_hair_L_02_ctrl\"),Transform=(Rotation=(X=0,Y=0,Z=0,W=1),Translation=(X=0,Y=0,Z=1.5),Scale3D=(X=1,Y=1,Z=1))),(Item=(Type=Control,Name=\"dyn_hair_M_02_ctrl\"),Transform=(Rotation=(X=0,Y=0,Z=0,W=1),Translation=(X=0,Y=0,Z=1.5),Scale ...(truncated)"
    WeightMinimum=0
    Weight=@DeltaRotation_4.outputWeight
   )
   n4 = RigVMVariableNode() {variable=pub isTargeting}
   > &RigUnit_ModifyTransforms(
    ItemToModify="((Item=(Type=Control,Name=\"dyn_hair_R_01_ctrl\"),Transform=(Rotation=(X=0,Y=0,Z=0,W=1),Translation=(X=0,Y=-4,Z=2),Scale3D=(X=1,Y=1,Z=1))),(Item=(Type=Control,Name=\"dyn_hair_R_02_ctrl\"),Transform=(Rotation=(X=0,Y=0,Z=0,W=1),Translation=(X=6,Y=-6,Z=6),Scale3D ...(truncated)"
    WeightMinimum=0
    Weight=&RigVMFunction_MathDoubleMul(B=&RigVMFunction_MathBoolToFloat(Value=@n4.Value).Result, A=@DeltaRotation_4.outputWeight).Result
   )
   > DeltaRotation_5 = RigVMFunctionReferenceNode(child=head, max=50, min=15, onlyTwist=true, parent=head_null_sp, parentType=Null, primAxis=0, rotationOrder=ZYX, twistAxis=(X=1,Y=0,Z=0))
   n5 = &RigVMFunction_MathDoubleMul(B=0.5, A=@DeltaRotation_5.outputWeight)
   > &RigUnit_ModifyTransforms(
    ItemToModify="((Item=(Type=Control,Name=\"base_hair_R_ctrl\"),Transform=(Rotation=(X=0,Y=0,Z=0,W=1),Translation=(X=0,Y=0,Z=2),Scale3D=(X=1,Y=1,Z=1))),(Item=(Type=Control,Name=\"dyn_hair_R_01_ctrl\"),Transform=(Rotation=(X=0,Y=0,Z=0,W=1),Translation=(X=0,Y=-1,Z=3),Scale3D=(X ...(truncated)"
    WeightMinimum=0
    Weight=@n5.Result
   )
   > &RigUnit_ModifyTransforms(
    ItemToModify="((Item=(Type=Control,Name=\"base_hair_M_ctrl\"),Transform=(Rotation=(X=0,Y=0,Z=0,W=1),Translation=(X=0,Y=0,Z=2),Scale3D=(X=1,Y=1,Z=1))),(Item=(Type=Control,Name=\"dyn_hair_M_01_ctrl\"),Transform=(Rotation=(X=0,Y=0,Z=0,W=1),Translation=(X=0,Y=2,Z=3),Scale3D=(X= ...(truncated)"
    WeightMinimum=0
    Weight=@n5.Result
   )
   > &RigUnit_ModifyTransforms(
    ItemToModify="((Item=(Type=Control,Name=\"base_hair_L_ctrl\"),Transform=(Rotation=(X=0,Y=0,Z=0,W=1),Translation=(X=0,Y=0,Z=2),Scale3D=(X=1,Y=1,Z=1))),(Item=(Type=Control,Name=\"dyn_hair_L_01_ctrl\"),Transform=(Rotation=(X=0,Y=0,Z=0,W=1),Translation=(X=0,Y=4,Z=0),Scale3D=(X= ...(truncated)"
    WeightMinimum=0
    Weight=@n5.Result
   )
   > DeltaRotation_6 = RigVMFunctionReferenceNode(child=head, max=-50, min=-15, onlyTwist=true, parent=head_null_sp, parentType=Null, primAxis=0, rotationOrder=ZYX, twistAxis=(X=1,Y=0,Z=0))
   n6 = RigVMVariableNode() {variable=pub isTargeting}
   n7 = &RigVMFunction_MathDoubleMul(
    A=@DeltaRotation_6.outputWeight
    B=&RigVMFunction_MathBoolToFloat(Value=&RigVMFunction_MathBoolNot(Value=@n6.Value).Result).Result
   )
   > &RigUnit_ModifyTransforms(
    ItemToModify="((Item=(Type=Control,Name=\"base_hair_R_ctrl\"),Transform=(Rotation=(X=0,Y=0,Z=0,W=1),Translation=(X=0,Y=-1,Z=0),Scale3D=(X=1,Y=1,Z=1))),(Item=(Type=Control,Name=\"dyn_hair_R_01_ctrl\"),Transform=(Rotation=(X=0,Y=0,Z=0,W=1),Translation=(X=1,Y=-2,Z=2),Scale3D=( ...(truncated)"
    WeightMinimum=0
    Weight=@n7.Result
   )
   > &RigUnit_ModifyTransforms(
    ItemToModify="((Item=(Type=Control,Name=\"base_hair_M_ctrl\"),Transform=(Rotation=(X=0,Y=0,Z=0,W=1),Translation=(X=0,Y=1,Z=1),Scale3D=(X=1,Y=1,Z=1))),(Item=(Type=Control,Name=\"dyn_hair_M_01_ctrl\"),Transform=(Rotation=(X=0,Y=0,Z=0,W=1),Translation=(X=0,Y=0.5,Z=2),Scale3D=( ...(truncated)"
    WeightMinimum=0
    Weight=@n7.Result
   )
   > &RigUnit_ModifyTransforms(
    ItemToModify="((Item=(Type=Control,Name=\"base_hair_L_ctrl\"),Transform=(Rotation=(X=0,Y=0,Z=0,W=1),Translation=(X=1,Y=1,Z=0),Scale3D=(X=1,Y=1,Z=1))),(Item=(Type=Control,Name=\"dyn_hair_L_01_ctrl\"),Transform=(Rotation=(X=0,Y=0,Z=0,W=1),Translation=(X=1.5,Y=-0.5,Z=1),Scale3 ...(truncated)"
    WeightMinimum=0
    Weight=@n7.Result
   )
   n8 = RigVMVariableNode() {variable=isCrouching}
   Multiply_12 = RigVMAggregateNode(
    A=@DeltaRotation_6.outputWeight
    C=&RigVMFunction_MathBoolToFloat(Value=RigVMVariableNode() {variable=pub isTargeting}.Value).Result
    D=&RigVMFunction_MathBoolToFloat(Value=&RigVMFunction_MathBoolNot(Value=@n8.Value).Result).Result
    E=RigVMVariableNode() {variable=pub isAssualtedRifleEquipped}.Value
   )
   > &RigUnit_ModifyTransforms(
    ItemToModify=((Item=(Type=Bone,Name="dyn_hair_R_07"),Transform=(Rotation=(X=0,Y=-0.478,Z=0,W=0.8784),Translation=(X=-1.8391,Y=-1.891,Z=0),Scale3D=(X=1,Y=1,Z=1))))
    WeightMinimum=0
    Weight=@Multiply_12.Result
   )
   n9 = &RigVMFunction_MathDoubleMul(A=@n7.Result, B=@DeltaRotation_4.outputWeight)
   > &RigUnit_ModifyTransforms(
    ItemToModify="((Item=(Type=Control,Name=\"base_hair_R_ctrl\"),Transform=(Rotation=(X=0,Y=0,Z=0,W=1),Translation=(X=0,Y=-2,Z=-2),Scale3D=(X=1,Y=1,Z=1))),(Item=(Type=Control,Name=\"dyn_hair_R_01_ctrl\"),Transform=(Rotation=(X=0,Y=0,Z=0,W=1),Translation=(X=0,Y=-3,Z=-10),Scale3 ...(truncated)"
    WeightMinimum=0
    Weight=@n9.Result
   )
   > &RigUnit_ModifyTransforms(
    ItemToModify="((Item=(Type=Control,Name=\"base_hair_M_ctrl\"),Transform=(Rotation=(X=0,Y=0,Z=0,W=1),Translation=(X=0,Y=0,Z=2),Scale3D=(X=1,Y=1,Z=1))),(Item=(Type=Control,Name=\"dyn_hair_M_01_ctrl\"),Transform=(Rotation=(X=0,Y=0,Z=0,W=1),Translation=(X=0,Y=0,Z=-4),Scale3D=(X ...(truncated)"
    WeightMinimum=0
    Weight=@n9.Result
   )
   > &RigUnit_ModifyTransforms(
    ItemToModify="((Item=(Type=Control,Name=\"base_hair_L_ctrl\"),Transform=(Rotation=(X=0,Y=0,Z=0,W=1),Translation=(X=1,Y=1,Z=0),Scale3D=(X=1,Y=1,Z=1))),(Item=(Type=Control,Name=\"dyn_hair_L_01_ctrl\"),Transform=(Rotation=(X=0,Y=0,Z=0,W=1),Translation=(X=1.5,Y=-0.5,Z=-1),Scale ...(truncated)"
    WeightMinimum=0
    Weight=@n9.Result
   )
   > DeltaRotation_7 = RigVMFunctionReferenceNode(child=head, max=-100, min=-60, parent=head_null_sp, parentType=Null, primAxis=2, rotationOrder=ZYX, twistAxis=(X=1,Y=0,Z=0))
   > &RigUnit_ModifyTransforms(
    ItemToModify="((Item=(Type=Control,Name=\"base_hair_L_ctrl\"),Transform=(Rotation=(X=0,Y=-0.1736,Z=0,W=0.9848),Translation=(X=0,Y=0,Z=1),Scale3D=(X=1,Y=1,Z=1))),(Item=(Type=Control,Name=\"dyn_hair_L_01_ctrl\"),Transform=(Rotation=(X=0,Y=0,Z=0,W=1),Translation=(X=0,Y=0,Z=2), ...(truncated)"
    WeightMinimum=0
    Weight=@DeltaRotation_7.outputWeight
   )
   > &RigUnit_ModifyTransforms(
    ItemToModify="((Item=(Type=Control,Name=\"base_hair_M_ctrl\"),Transform=(Rotation=(X=0,Y=-0.1736,Z=0,W=0.9848),Translation=(X=0,Y=0,Z=1),Scale3D=(X=1,Y=1,Z=1))),(Item=(Type=Control,Name=\"dyn_hair_M_01_ctrl\"),Transform=(Rotation=(X=0,Y=0,Z=0,W=1),Translation=(X=0,Y=0,Z=2), ...(truncated)"
    WeightMinimum=0
    Weight=@DeltaRotation_7.outputWeight
   )
   > &RigUnit_ModifyTransforms(
    ItemToModify="((Item=(Type=Control,Name=\"base_hair_R_ctrl\"),Transform=(Rotation=(X=0,Y=-0.1736,Z=0,W=0.9848),Translation=(X=0,Y=0,Z=1),Scale3D=(X=1,Y=1,Z=1))),(Item=(Type=Control,Name=\"dyn_hair_R_01_ctrl\"),Transform=(Rotation=(X=0,Y=0,Z=0,W=1),Translation=(X=0,Y=0,Z=2), ...(truncated)"
    WeightMinimum=0
    Weight=@DeltaRotation_7.outputWeight
   )
   Clamp = &RigVMFunction_MathDoubleClamp(Maximum=0.55, Value=RigVMVariableNode() {variable=isHeadUp}.Value)
   RigVMFunction_MathBoolToFloat = &RigVMFunction_MathBoolToFloat(Value=RigVMVariableNode() {variable=pub isTargeting}.Value)
   n10 = &RigVMFunction_AnimEvalRichCurve(
    Curve=(EditorCurveData=(Keys=((InterpMode=RCIM_Cubic),(InterpMode=RCIM_Cubic,Time=0.3879,Value=0.174,ArriveTangent=1,LeaveTangent=1),(InterpMode=RCIM_Cubic,TangentMode=RCTM_User,Time=1,Value=1,ArriveTangent=1.9233,LeaveTangent=1.9233))))
    SourceMinimum=0
    TargetMinimum=0
    Value=&RigVMFunction_MathDoubleMul(B=@Clamp.Result, A=@RigVMFunction_MathBoolToFloat.Result).Result
   )
   n11 = &RigVMFunction_MathDoubleMul(B=0.8, A=@n10.Result)
   > &RigUnit_ModifyTransforms(
    ItemToModify="((Item=(Type=Control,Name=\"base_hair_R_ctrl\"),Transform=(Rotation=(X=0,Y=0,Z=0,W=1),Translation=(X=0,Y=0,Z=1),Scale3D=(X=1,Y=1,Z=1))),(Item=(Type=Control,Name=\"dyn_hair_R_01_ctrl\"),Transform=(Rotation=(X=-0.5,Y=0,Z=0,W=0.866),Translation=(X=0,Y=-5,Z=6),Sca ...(truncated)"
    WeightMinimum=0
    Weight=@n11.Result
   )
   > &RigUnit_ModifyTransforms(
    ItemToModify="((Item=(Type=Control,Name=\"dyn_hair_M_01_ctrl\"),Transform=(Rotation=(X=0,Y=-0.1736,Z=0,W=0.9848),Translation=(X=0,Y=0,Z=1),Scale3D=(X=1,Y=1,Z=1))),(Item=(Type=Control,Name=\"dyn_hair_M_01_ctrl\"),Transform=(Rotation=(X=0,Y=0,Z=0,W=1),Translation=(X=0,Y=-3,Z= ...(truncated)"
    WeightMinimum=0
    Weight=@n11.Result
   )
   RigVMFunction_MathBoolNot = &RigVMFunction_MathBoolNot(Value=RigVMVariableNode() {variable=isWearingBackBling}.Value)
   RigVMFunction_MathBoolAnd = &RigVMFunction_MathBoolAnd(A=@RigVMFunction_MathBoolNot.Result, B=RigVMVariableNode() {variable=pub isDiving}.Value)
   > RigVMVariableNode(Value=@RigVMFunction_MathBoolAnd.Result) {variable=isDivingNoBackBling}
   n12 = &RigVMFunction_MathBoolToFloat(Value=RigVMVariableNode() {variable=isDivingNoBackBling}.Value)
   > &RigUnit_ModifyTransforms(
    ItemToModify="((Item=(Type=Control,Name=\"dyn_hair_L_01_ctrl\"),Transform=(Rotation=(X=0,Y=0,Z=0,W=1),Translation=(X=0,Y=0,Z=1),Scale3D=(X=1,Y=1,Z=1))),(Item=(Type=Control,Name=\"dyn_hair_L_02_ctrl\"),Transform=(Rotation=(X=0,Y=0,Z=0,W=1),Translation=(X=0,Y=0,Z=2),Scale3D=( ...(truncated)"
    WeightMinimum=0
    Weight=@n12.Result
   )
   > &RigUnit_ModifyTransforms(
    ItemToModify="((Item=(Type=Control,Name=\"dyn_hair_M_01_ctrl\"),Transform=(Rotation=(X=0,Y=0,Z=0,W=1),Translation=(X=0,Y=0,Z=1),Scale3D=(X=1,Y=1,Z=1))),(Item=(Type=Control,Name=\"dyn_hair_M_02_ctrl\"),Transform=(Rotation=(X=0,Y=0,Z=0,W=1),Translation=(X=0,Y=0,Z=2),Scale3D=( ...(truncated)"
    WeightMinimum=0
    Weight=@n12.Result
   )
   > &RigUnit_ModifyTransforms(
    ItemToModify="((Item=(Type=Control,Name=\"dyn_hair_R_01_ctrl\"),Transform=(Rotation=(X=0,Y=0,Z=0,W=1),Translation=(X=0,Y=0,Z=1),Scale3D=(X=1,Y=1,Z=1))),(Item=(Type=Control,Name=\"dyn_hair_R_02_ctrl\"),Transform=(Rotation=(X=0,Y=0,Z=0,W=1),Translation=(X=0,Y=0,Z=2),Scale3D=( ...(truncated)"
    WeightMinimum=0
    Weight=@n12.Result
   )
   n13 = RigVMVariableNode() {variable=IsPlayCelebrateEmote}
   > &RigUnit_ModifyTransforms(
    ItemToModify="((Item=(Type=Control,Name=\"dyn_hair_M_02_ctrl\"),Transform=(Rotation=(X=0,Y=0,Z=0,W=1),Translation=(X=0,Y=0,Z=3),Scale3D=(X=1,Y=1,Z=1))),(Item=(Type=Control,Name=\"dyn_hair_M_03_ctrl\"),Transform=(Rotation=(X=0,Y=0,Z=0,W=1),Translation=(X=0,Y=0,Z=2),Scale3D=( ...(truncated)"
    WeightMinimum=0
    Weight=@n13.Value
   )
   > &RigUnit_ModifyTransforms(
    ItemToModify="((Item=(Type=Control,Name=\"dyn_hair_L_02_ctrl\"),Transform=(Rotation=(X=0,Y=0,Z=0,W=1),Translation=(X=0,Y=0,Z=3),Scale3D=(X=1,Y=1,Z=1))),(Item=(Type=Control,Name=\"dyn_hair_L_03_ctrl\"),Transform=(Rotation=(X=0,Y=0,Z=0,W=1),Translation=(X=0,Y=0,Z=2),Scale3D=( ...(truncated)"
    WeightMinimum=0
    Weight=@n13.Value
   )
   > &RigUnit_ModifyTransforms(
    ItemToModify="((Item=(Type=Control,Name=\"dyn_hair_R_02_ctrl\"),Transform=(Rotation=(X=0,Y=0,Z=0,W=1),Translation=(X=0,Y=0,Z=3),Scale3D=(X=1,Y=1,Z=1))),(Item=(Type=Control,Name=\"dyn_hair_R_03_ctrl\"),Transform=(Rotation=(X=0,Y=0,Z=0,W=1),Translation=(X=0,Y=0,Z=2),Scale3D=( ...(truncated)"
    WeightMinimum=0
    Weight=@n13.Value
   )
   n14 = RigVMVariableNode() {variable=isDBNO}
   n15 = &RigVMFunction_MathBoolToFloat(Value=@n14.Value)
   > &RigUnit_ModifyTransforms(
    ItemToModify="((Item=(Type=Control,Name=\"dyn_hair_M_02_ctrl\"),Transform=(Rotation=(X=0,Y=0,Z=0,W=1),Translation=(X=0,Y=0,Z=3),Scale3D=(X=1,Y=1,Z=1))),(Item=(Type=Control,Name=\"dyn_hair_M_03_ctrl\"),Transform=(Rotation=(X=0,Y=0,Z=0,W=1),Translation=(X=0,Y=0,Z=2),Scale3D=( ...(truncated)"
    WeightMinimum=0
    Weight=@n15.Result
   )
   > &RigUnit_ModifyTransforms(
    ItemToModify="((Item=(Type=Control,Name=\"dyn_hair_L_02_ctrl\"),Transform=(Rotation=(X=0,Y=0,Z=0,W=1),Translation=(X=0,Y=0,Z=1),Scale3D=(X=1,Y=1,Z=1))),(Item=(Type=Control,Name=\"dyn_hair_L_03_ctrl\"),Transform=(Rotation=(X=0,Y=0,Z=0,W=1),Translation=(X=0,Y=0,Z=1),Scale3D=( ...(truncated)"
    WeightMinimum=0
    Weight=@n15.Result
   )
   > &RigUnit_ModifyTransforms(
    ItemToModify="((Item=(Type=Control,Name=\"dyn_hair_R_02_ctrl\"),Transform=(Rotation=(X=0,Y=0,Z=0,W=1),Translation=(X=0,Y=2,Z=1),Scale3D=(X=1,Y=1,Z=1))),(Item=(Type=Control,Name=\"dyn_hair_R_03_ctrl\"),Transform=(Rotation=(X=0,Y=0,Z=0,W=1),Translation=(X=0,Y=1,Z=3),Scale3D=( ...(truncated)"
    WeightMinimum=0
    Weight=@n15.Result
   )
   n16 = RigVMVariableNode() {variable=ismoving2D}
   > &RigUnit_ModifyTransforms(
    ItemToModify="((Item=(Type=Control,Name=\"dyn_hair_R_03_ctrl\"),Transform=(Rotation=(X=0,Y=0,Z=0,W=1),Translation=(X=0,Y=2,Z=-3),Scale3D=(X=1,Y=1,Z=1))),(Item=(Type=Control,Name=\"dyn_hair_M_03_ctrl\"),Transform=(Rotation=(X=0,Y=0,Z=0,W=1),Translation=(X=0,Y=0,Z=-2),Scale3D ...(truncated)"
    WeightMinimum=0
    Weight=&RigVMFunction_MathBoolToFloat(Value=&RigVMFunction_MathBoolAnd(B=@n16.Value, A=@n14.Value).Result).Result
   )
   > RigVMFunctionReturnNode()
  ]}
 New Function_ContainedGraph {nodes [
   RigVMFunctionEntryNode()
   RigVMFunction_MathBoolOr = &RigVMFunction_MathBoolOr(A=RigVMVariableNode() {variable=isCrouching}.Value, B=RigVMVariableNode() {variable=pub isTargeting}.Value)
   > RigVMVariableNode(Value=@RigVMFunction_MathBoolOr.Result) {variable=ClippingCorrect}
   Multiply = &RigVMFunction_MathDoubleMul(B=0.8, A=&RigVMFunction_MathBoolToFloat(Value=RigVMVariableNode() {variable=ClippingCorrect}.Value).Result)
   > &RigUnit_ModifyTransforms(
    ItemToModify="((Item=(Type=Bone,Name=\"dyn_hair_R_07\"),Transform=(Rotation=(X=0,Y=0.0872,Z=0,W=0.9962),Translation=(X=2,Y=0,Z=6),Scale3D=(X=1,Y=1,Z=1))),(Item=(Type=Bone,Name=\"dyn_hair_L_07\"),Transform=(Rotation=(X=0,Y=0,Z=0,W=1),Translation=(X=0,Y=0,Z=8),Scale3D=(X=1,Y= ...(truncated)"
    WeightMinimum=0
    Weight=@Multiply.Result
   )
   > &RigUnit_ModifyTransforms(
    ItemToModify=((Item=(Type=Bone,Name="dyn_hair_R_07"),Transform=(Rotation=(X=0,Y=0,Z=0,W=1),Translation=(X=-8.1688,Y=1.869,Z=2.3736),Scale3D=(X=1,Y=1,Z=1))))
    WeightMinimum=0
    Weight=RigVMVariableNode() {variable=isHeadUp}.Value
   )
   Multiply_4 = RigVMAggregateNode(
    B=&RigVMFunction_MathBoolToFloat(Value=RigVMVariableNode() {variable=pub isTargeting}.Value).Result
    A=&RigVMFunction_MathBoolToFloat(Value=RigVMVariableNode() {variable=isCrouching}.Value).Result
    C=RigVMVariableNode() {variable=isHeadUp}.Value
   )
   > &RigUnit_ModifyTransforms(
    ItemToModify="((Item=(Type=Bone,Name=\"dyn_hair_R_07\"),Transform=(Rotation=(X=0,Y=0,Z=0,W=1),Translation=(X=7.5238,Y=0,Z=5.3346),Scale3D=(X=1,Y=1,Z=1))),(Item=(Type=Bone,Name=\"dyn_hair_R_02\"),Transform=(Rotation=(X=0,Y=0,Z=0,W=1),Translation=(X=0,Y=0,Z=2),Scale3D=(X=1,Y= ...(truncated)"
    WeightMinimum=0
    Weight=@Multiply_4.Result
   )
   RigVMFunction_MathBoolNot = &RigVMFunction_MathBoolNot(Value=RigVMVariableNode() {variable=isCrouching}.Value)
   Multiply_1 = RigVMAggregateNode(
    A=&RigVMFunction_MathBoolToFloat(Value=RigVMVariableNode() {variable=pub isTargeting}.Value).Result
    C=&RigVMFunction_MathBoolToFloat(Value=@RigVMFunction_MathBoolNot.Result).Result
    B=RigVMVariableNode() {variable=isHeadUp}.Value
   )
   > &RigUnit_ModifyTransforms(
    ItemToModify=((Item=(Type=Bone,Name="dyn_hair_R_07"),Transform=(Rotation=(X=0,Y=0,Z=0,W=1),Translation=(X=0,Y=0,Z=5),Scale3D=(X=1,Y=1,Z=1))))
    WeightMinimum=0
    Weight=@Multiply_1.Result
   )
   n1 = RigVMVariableNode() {variable=pub isTargeting}
   Multiply_2 = RigVMAggregateNode(
    A=&RigVMFunction_MathBoolToFloat(Value=RigVMVariableNode() {variable=isCrouching}.Value).Result
    C=&RigVMFunction_MathBoolToFloat(Value=&RigVMFunction_MathBoolNot(Value=@n1.Value).Result).Result
    B=RigVMVariableNode() {variable=isHeadUp}.Value
   )
   > &RigUnit_ModifyTransforms(
    ItemToModify=((Item=(Type=Bone,Name="dyn_hair_R_07"),Transform=(Rotation=(X=0,Y=0.3451,Z=0,W=0.9386),Translation=(X=-6.7569,Y=0,Z=5.8278),Scale3D=(X=1,Y=1,Z=1))))
    WeightMinimum=0
    Weight=@Multiply_2.Result
   )
   > &RigUnit_ModifyTransforms(
    ItemToModify=((Item=(Type=Bone,Name="dyn_hair_R_07"),Transform=(Rotation=(X=0,Y=0,Z=0,W=1),Translation=(X=0,Y=0,Z=-5),Scale3D=(X=1,Y=1,Z=1))))
    WeightMinimum=0
    Weight=&RigVMFunction_MathBoolToFloat(Value=RigVMVariableNode() {variable=isPistolEquiped}.Value).Result
   )
   Multiply_3 = RigVMAggregateNode(
    A=&RigVMFunction_MathBoolToFloat(Value=RigVMVariableNode() {variable=isCrouching}.Value).Result
    D=&RigVMFunction_MathBoolToFloat(Value=RigVMVariableNode() {variable=isPistolEquiped}.Value).Result
    B=&RigVMFunction_MathBoolToFloat(Value=RigVMVariableNode() {variable=pub isTargeting}.Value).Result
    C=RigVMVariableNode() {variable=isHeadUp}.Value
   )
   > &RigUnit_ModifyTransforms(
    ItemToModify="((Item=(Type=Bone,Name=\"dyn_hair_R_07\"),Transform=(Rotation=(X=0,Y=0.2588,Z=0,W=0.9659),Translation=(X=0,Y=0,Z=6),Scale3D=(X=1,Y=1,Z=1))),(Item=(Type=Bone,Name=\"dyn_hair_R_02\"),Transform=(Rotation=(X=0.0226,Y=0.2578,Z=-0.0842,W=0.9623),Translation=(X=1,Y=0 ...(truncated)"
    WeightMinimum=0
    Weight=@Multiply_3.Result
   )
   n2 = RigVMVariableNode() {variable=isCrouching}
   Multiply_5 = RigVMAggregateNode(
    B=&RigVMFunction_MathBoolToFloat(Value=RigVMVariableNode() {variable=pub isTargeting}.Value).Result
    D=&RigVMFunction_MathBoolToFloat(Value=RigVMVariableNode() {variable=isPistolEquiped}.Value).Result
    C=RigVMVariableNode() {variable=isHeadUp}.Value
    A=&RigVMFunction_MathBoolToFloat(Value=&RigVMFunction_MathBoolNot(Value=@n2.Value).Result).Result
   )
   > &RigUnit_ModifyTransforms(
    ItemToModify=((Item=(Type=Bone,Name="dyn_hair_R_07"),Transform=(Rotation=(X=0,Y=0.342,Z=0,W=0.9397),Translation=(X=0,Y=0,Z=-6),Scale3D=(X=1,Y=1,Z=1))))
    WeightMinimum=0
    Weight=@Multiply_5.Result
   )
   RigVMFunction_MathBoolAnd = &RigVMFunction_MathBoolAnd(
    A=&RigVMFunction_MathBoolNot(Value=RigVMVariableNode() {variable=isWearingBackBling}.Value).Result
    B=&RigVMFunction_MathBoolOr(A=RigVMVariableNode() {variable=pub diveFloating}.Value, B=RigVMVariableNode() {variable=pub isDiving}.Value).Result
   )
   > &RigUnit_ModifyTransforms(
    ItemToModify="((Item=(Type=Bone,Name=\"dyn_hair_R_07\"),Transform=(Rotation=(X=0,Y=0,Z=-0.5,W=0.866),Translation=(X=0,Y=4,Z=6),Scale3D=(X=1,Y=1,Z=1))),(Item=(Type=Bone,Name=\"dyn_hair_L_07\"),Transform=(Rotation=(X=0,Y=0,Z=0.342,W=0.9397),Translation=(X=0,Y=-6,Z=12),Scale3D ...(truncated)"
    WeightMinimum=0
    Weight=&RigVMFunction_MathBoolToFloat(Value=@RigVMFunction_MathBoolAnd.Result).Result
   )
   Multiply_7 = RigVMAggregateNode(
    C=&RigVMFunction_MathBoolToFloat(Value=RigVMVariableNode() {variable=pub isTargeting}.Value).Result
    A=&RigVMFunction_MathBoolToFloat(Value=RigVMVariableNode() {variable=isCrouching}.Value).Result
    D=&RigVMFunction_MathBoolToFloat(Value=RigVMVariableNode() {variable=ismoving2D}.Value).Result
   )
   > &RigUnit_ModifyTransforms(
    ItemToModify="((Item=(Type=Bone,Name=\"dyn_hair_R_07\"),Transform=(Rotation=(X=0,Y=0,Z=0,W=1),Translation=(X=0,Y=0,Z=-8.5),Scale3D=(X=1,Y=1,Z=1))),(Item=(Type=Bone,Name=\"dyn_hair_L_07\"),Transform=(Rotation=(X=0,Y=0,Z=0,W=1),Translation=(X=0,Y=0,Z=-7),Scale3D=(X=1,Y=1,Z=1) ...(truncated)"
    WeightMinimum=0
    Weight=@Multiply_7.Result
   )
   n3 = RigVMVariableNode() {variable=pub isTargeting}
   Multiply_8 = RigVMAggregateNode(
    A=&RigVMFunction_MathBoolToFloat(Value=RigVMVariableNode() {variable=isCrouching}.Value).Result
    D=&RigVMFunction_MathBoolToFloat(Value=RigVMVariableNode() {variable=ismoving2D}.Value).Result
    E=&RigVMFunction_MathBoolToFloat(Value=&RigVMFunction_MathBoolNot(Value=@n3.Value).Result).Result
   )
   > &RigUnit_ModifyTransforms(
    ItemToModify=((Item=(Type=Bone,Name="dyn_hair_R_07"),Transform=(Rotation=(X=0,Y=0,Z=0,W=1),Translation=(X=0,Y=0,Z=-5),Scale3D=(X=1,Y=1,Z=1))))
    WeightMinimum=0
    Weight=@Multiply_8.Result
   )
   n4 = RigVMVariableNode() {variable=isCrouching}
   Multiply_9 = RigVMAggregateNode(
    D=&RigVMFunction_MathBoolToFloat(Value=RigVMVariableNode() {variable=pub isTargeting}.Value).Result
    F=&RigVMFunction_MathBoolToFloat(Value=&RigVMFunction_MathBoolNot(Value=@n4.Value).Result).Result
    A=RigVMVariableNode() {variable=isHeadUp}.Value
    E=RigVMVariableNode() {variable=pub isAssualtedRifleEquipped}.Value
   )
   > &RigUnit_ModifyTransforms(
    ItemToModify="((Item=(Type=Bone,Name=\"dyn_hair_R_04\"),Transform=(Rotation=(X=0,Y=0,Z=0,W=1),Translation=(X=0,Y=0,Z=-8),Scale3D=(X=1,Y=1,Z=1))),(Item=(Type=Bone,Name=\"dyn_hair_R_03\"),Transform=(Rotation=(X=0,Y=0,Z=0,W=1),Translation=(X=-4,Y=0,Z=-1),Scale3D=(X=1,Y=1,Z=1)) ...(truncated)"
    WeightMinimum=0
    Weight=@Multiply_9.Result
   )
   n5 = RigVMVariableNode() {variable=ismoving2D}
   Multiply_10 = RigVMAggregateNode(
    D=&RigVMFunction_MathBoolToFloat(Value=RigVMVariableNode() {variable=pub isTargeting}.Value).Result
    F=&RigVMFunction_MathBoolToFloat(Value=RigVMVariableNode() {variable=isCrouching}.Value).Result
    G=&RigVMFunction_MathBoolToFloat(Value=&RigVMFunction_MathBoolNot(Value=@n5.Value).Result).Result
    A=RigVMVariableNode() {variable=isHeadUp}.Value
    E=RigVMVariableNode() {variable=pub isAssualtedRifleEquipped}.Value
   )
   > &RigUnit_ModifyTransforms(
    ItemToModify="((Item=(Type=Bone,Name=\"dyn_hair_R_04\"),Transform=(Rotation=(X=0,Y=0,Z=0,W=1),Translation=(X=0,Y=0,Z=-1),Scale3D=(X=1,Y=1,Z=1))),(Item=(Type=Bone,Name=\"dyn_hair_R_03\"),Transform=(Rotation=(X=0,Y=0,Z=0,W=1),Translation=(X=-6,Y=0,Z=-6),Scale3D=(X=1,Y=1,Z=1)) ...(truncated)"
    WeightMinimum=0
    Weight=@Multiply_10.Result
   )
   n6 = RigVMVariableNode() {variable=isPistolEquiped}
   Multiply_11 = RigVMAggregateNode(
    D=&RigVMFunction_MathBoolToFloat(Value=RigVMVariableNode() {variable=pub isTargeting}.Value).Result
    F=&RigVMFunction_MathBoolToFloat(Value=RigVMVariableNode() {variable=isCrouching}.Value).Result
    G=&RigVMFunction_MathBoolToFloat(Value=RigVMVariableNode() {variable=ismoving2D}.Value).Result
    E=&RigVMFunction_MathBoolToFloat(Value=&RigVMFunction_MathBoolNot(Value=@n6.Value).Result).Result
    A=RigVMVariableNode() {variable=isHeadUp}.Value
   )
   > &RigUnit_ModifyTransforms(
    ItemToModify="((Item=(Type=Bone,Name=\"dyn_hair_R_04\"),Transform=(Rotation=(X=0,Y=0,Z=0,W=1),Translation=(X=0,Y=0,Z=-6),Scale3D=(X=1,Y=1,Z=1))),(Item=(Type=Bone,Name=\"dyn_hair_R_03\"),Transform=(Rotation=(X=0,Y=0,Z=0,W=1),Translation=(X=10,Y=0,Z=-8),Scale3D=(X=1,Y=1,Z=1)) ...(truncated)"
    WeightMinimum=0
    Weight=@Multiply_11.Result
   )
   n7 = RigVMVariableNode() {variable=pub isTargeting}
   n8 = RigVMVariableNode() {variable=isCrouching}
   > &RigUnit_ModifyTransforms(
    ItemToModify="((Item=(Type=Bone,Name=\"dyn_hair_R_07\"),Transform=(Rotation=(X=0,Y=0.1736,Z=0,W=0.9848),Translation=(X=0,Y=0,Z=3),Scale3D=(X=1,Y=1,Z=1))),(Item=(Type=Bone,Name=\"dyn_hair_R_02\"),Transform=(Rotation=(X=0,Y=0,Z=0,W=1),Translation=(X=0,Y=0,Z=2),Scale3D=(X=1,Y= ...(truncated)"
    WeightMinimum=0
    Weight=&RigVMFunction_MathDoubleMul(A=&RigVMFunction_MathBoolToFloat(Value=@n7.Value).Result, B=&RigVMFunction_MathBoolToFloat(Value=@n8.Value).Result).Result
   )
   > RigVMFunctionReturnNode()
  ]}
 RigVMFunctionLibrary {nodes [
   AlterTransform = RigVMCollapseNode(ExecuteContext=(), inheristTransform=False, transforms=())
   ClippingCorrections = RigVMCollapseNode(ExecuteContext=())
   Constraints = RigVMCollapseNode(ExecuteContext=())
   Controllers = RigVMCollapseNode(Debug=False, ExecuteContext=())
   DeltaRotation = RigVMCollapseNode(ExecuteContext=(), twistAxis=(X=0,Y=0,Z=0))
   HairAdjustments = RigVMCollapseNode(ExecuteContext=())
   HeadAdjustments = RigVMCollapseNode(ExecuteContext=())
   claviclesBackForward = RigVMCollapseNode(ExecuteContext=())
   splineIK = RigVMCollapseNode(ExecuteContext=())
  ]}
 RigVMFunction_Sequence_1_ContainedGraph {nodes [
   n1 = RigVMFunctionEntryNode()
   RigVMFunction_Sequence = &RigVMFunction_Sequence(ExecuteContext=@n1.ExecuteContext)
   n2 = &RigVMFunction_Sequence(ExecuteContext=@RigVMFunction_Sequence.B)
   RigVMFunctionReturnNode(A=@RigVMFunction_Sequence.A, B=@n2.A, C=@n2.B)
  ]}
 RigVMFunction_Sequence_1_ContainedGraph {nodes [
   n1 = RigVMFunctionEntryNode()
   RigVMFunction_Sequence = &RigVMFunction_Sequence(ExecuteContext=@n1.ExecuteContext)
   n2 = &RigVMFunction_Sequence(ExecuteContext=@RigVMFunction_Sequence.B)
   RigVMFunctionReturnNode(A=@RigVMFunction_Sequence.A, B=@n2.A, D=@n2.B)
  ]}
 RigVMFunction_Sequence_1_ContainedGraph {nodes [
   n1 = RigVMFunctionEntryNode()
   RigVMFunction_Sequence = &RigVMFunction_Sequence(ExecuteContext=@n1.ExecuteContext)
   n2 = &RigVMFunction_Sequence(ExecuteContext=@RigVMFunction_Sequence.B)
   n3 = &RigVMFunction_Sequence(ExecuteContext=@n2.B)
   RigVMFunctionReturnNode(A=@RigVMFunction_Sequence.A, B=@n2.A, C=@n3.A, D=@n3.B)
  ]}
 RigVMFunction_Sequence_3_ContainedGraph {nodes [
   n1 = RigVMFunctionEntryNode()
   n2 = &RigVMFunction_Sequence(ExecuteContext=@n1.ExecuteContext)
   n3 = &RigVMFunction_Sequence(ExecuteContext=@n2.B)
   RigVMFunctionReturnNode(A=@n2.A, B=@n3.A, C=@n3.B)
  ]}
 RigVMModel {nodes [
   RigUnit_BeginExecution = &RigUnit_BeginExecution()
   splineIK = RigVMFunctionReferenceNode(ExecuteContext=@RigUnit_BeginExecution.ExecutePin)
   > HairAdjustments = RigVMFunctionReferenceNode()
   > ClippingCorrections = RigVMFunctionReferenceNode()
   Controllers = RigVMFunctionReferenceNode(Debug=true, ExecuteContext=())
   RigVMVariableNode() {variable=pub disablecontrolRig}
  ]}
 RigVMModel Construction Event Graph {nodes [
   n1 = &RigUnit_PrepareForExecution()
   RigVMFunction_Sequence = RigVMAggregateNode(ExecuteContext=@n1.ExecutePin)
   n2 = &RigUnit_ItemArray(Items="((Type=Control,Name=\"base_hair_R_ctrl\"),(Type=Control,Name=\"dyn_hair_R_01_ctrl\"),(Type=Control,Name=\"dyn_hair_R_02_ctrl\"),(Type=Control,Name=\"dyn_hair_R_03_ctrl\"),(Type=Control,Name=\"dyn_hair_R_04_ctrl\"),(Type=Control,Name=\"dyn_hair_R_05_ctrl\"),(Type=Contr ...(truncated)")
   RigVMVariableNode(Value=@n2.Items, ExecuteContext=@RigVMFunction_Sequence.B) {variable=dyn_hair_r_controls}
   > RigVMVariableNode(Value=/Script/ControlRig.RigUnit_HierarchyGetPoseItemArray(ElementType=All, Initial=False, ItemsToGet=@n2.Items) {is=struct RigVMUnitNode}.Pose) {variable=dyn_hair_r_pose}
   n3 = &RigUnit_ItemArray(Items="((Type=Control,Name=\"base_hair_L_ctrl\"),(Type=Control,Name=\"dyn_hair_L_01_ctrl\"),(Type=Control,Name=\"dyn_hair_L_02_ctrl\"),(Type=Control,Name=\"dyn_hair_L_03_ctrl\"),(Type=Control,Name=\"dyn_hair_L_04_ctrl\"),(Type=Control,Name=\"dyn_hair_L_05_ctrl\"),(Type=Contr ...(truncated)")
   RigVMVariableNode(Value=@n3.Items, ExecuteContext=@RigVMFunction_Sequence.A) {variable=dyn_hair_l_controls}
   > RigVMVariableNode(Value=&RigUnit_HierarchyGetPoseItemArray(ElementType=All, Initial=False, ItemsToGet=@n3.Items).Pose) {variable=dyn_hair_l_pose}
   RigUnit_ItemArray = &RigUnit_ItemArray(Items="((Type=Control,Name=\"base_hair_M_ctrl\"),(Type=Control,Name=\"dyn_hair_M_01_ctrl\"),(Type=Control,Name=\"dyn_hair_M_02_ctrl\"),(Type=Control,Name=\"dyn_hair_M_03_ctrl\"),(Type=Control,Name=\"dyn_hair_M_04_ctrl\"),(Type=Control,Name=\"dyn_hair_M_05_ctrl\"),(Type=Contr ...(truncated)")
   RigVMVariableNode(Value=@RigUnit_ItemArray.Items, ExecuteContext=@RigVMFunction_Sequence.C) {variable=dyn_hair_m_controls}
   > RigVMVariableNode(Value=&RigUnit_HierarchyGetPoseItemArray(ElementType=All, Initial=False, ItemsToGet=@RigUnit_ItemArray.Items).Pose) {variable=dyn_hair_m_pose}
  ]}
]
hierarchy [
 root {loc=<0 0 0>, rot=<0 0 0 1>}
 pelvis {parent=root, loc=<0 -0.5301 98.0571>, rot=<0.0127 0.707 -0.0127 -0.707>}
 spine_01 {parent=pelvis, loc=<6.1208 -0.9328 0>, rot=<0 0 0.034 -0.9994>}
 spine_02 {parent=spine_01, loc=<6.5997 -0.2095 0>, rot=<0 0 0.0238 -0.9997>}
 spine_03 {parent=spine_02, loc=<6.4914 -0.1237 0>, rot=<0 0 0.0149 -0.9999>}
 spine_04 {parent=spine_03, loc=<6.9598 0.0794 0>, rot=<0 0 -0.0132 -0.9999>}
 spine_05 {parent=spine_04, loc=<6.9987 0 0>, rot=<0 0 -0.0015 -1>}
 neck_01 {parent=spine_05, loc=<11.3138 0.1529 0>, rot=<0 0 -0.2003 -0.9797>}
 neck_02 {parent=neck_01, loc=<4.5627 0 0>, rot=<0 0 0.0504 -0.9987>}
 head {parent=neck_02, loc=<5.3914 0 0>, rot=<0 0 0.115 -0.9934>}
 base_hair {parent=head, loc=<9.7087 0 0>, rot=<0 0 0 1>}
 base_hair_M {parent=base_hair, loc=<-4.9288 -5.9597 0>, rot=<0.122 -0.6965 0.6965 -0.122>}
 dyn_hair_M_01 {parent=base_hair_M, loc=<2.7713 -0.077 -0.0391>, rot=<0 -0.0174 0 -0.9998>}
 dyn_hair_M_02 {parent=dyn_hair_M_01, loc=<3.3978 0.077 -0.0574>, rot=<0 0 0 1>}
 dyn_hair_M_03 {parent=dyn_hair_M_02, loc=<7.2599 0 -0.0363>, rot=<0 -0.0459 0 -0.9989>}
 dyn_hair_M_04 {parent=dyn_hair_M_03, loc=<9.4558 0 0.0822>, rot=<0 -0.0327 0 -0.9995>}
 dyn_hair_M_05 {parent=dyn_hair_M_04, loc=<8.0915 0 -0.0093>, rot=<0 -0.0863 0 -0.9963>}
 dyn_hair_M_06 {parent=dyn_hair_M_05, loc=<7.7516 0 0.0713>, rot=<0 0 0 1>}
 base_hair_L {parent=base_hair, loc=<-5.1453 -5.9648 -4.8248>, rot=<0.0912 -0.6958 0.6959 -0.1525>}
 dyn_hair_L_01 {parent=base_hair_L, loc=<2.1382 -0.1816 0.1051>, rot=<0 -0.0174 0 -0.9998>}
 dyn_hair_L_02 {parent=dyn_hair_L_01, loc=<4.0356 0.1816 -0.1795>, rot=<0 0 0 1>}
 dyn_hair_L_03 {parent=dyn_hair_L_02, loc=<7.2599 0 -0.0363>, rot=<0 -0.0459 0 -0.9989>}
 dyn_hair_L_04 {parent=dyn_hair_L_03, loc=<9.4558 0 0.0822>, rot=<0 -0.0327 0 -0.9995>}
 dyn_hair_L_05 {parent=dyn_hair_L_04, loc=<8.0915 0 -0.0093>, rot=<0 -0.0863 0 -0.9963>}
 dyn_hair_L_06 {parent=dyn_hair_L_05, loc=<7.7516 0 0.0713>, rot=<0 0 0 1>}
 dyn_hair_L_07 {parent=dyn_hair_L_02, loc=<2.945 2.7076 -3.874>, rot=<0 0 0 1>}
 base_hair_R {parent=base_hair, loc=<-4.9576 -5.9648 4.8248>, rot=<0.1525 -0.6959 0.6958 -0.0912>}
 dyn_hair_R_01 {parent=base_hair_R, loc=<2.7511 0.0166 -0.0356>, rot=<0 -0.0174 0 -0.9998>}
 dyn_hair_R_02 {parent=dyn_hair_R_01, loc=<3.4181 -0.0166 -0.0602>, rot=<0 0 0 1>}
 dyn_hair_R_03 {parent=dyn_hair_R_02, loc=<7.2599 0 -0.0363>, rot=<0 -0.0459 0 -0.9989>}
 dyn_hair_R_04 {parent=dyn_hair_R_03, loc=<9.4558 0 0.0822>, rot=<0 -0.0327 0 -0.9995>}
 dyn_hair_R_05 {parent=dyn_hair_R_04, loc=<8.0915 0 -0.0093>, rot=<0 -0.0863 0 -0.9963>}
 dyn_hair_R_06 {parent=dyn_hair_R_05, loc=<7.7516 0 0.0713>, rot=<0 0 0 1>}
 dyn_hair_R_07 {parent=dyn_hair_R_02, loc=<3.1233 -2.6916 -3.9316>, rot=<0 0 0 1>}
 dyn_frontHair_05 {parent=base_hair, loc=<-13.5499 0.9381 7.2092>, rot=<0.0801 -0.0258 0.9719 -0.22>}
 FACIAL_C_FacialRoot {parent=head, loc=<2.1524 -0.746 -0.0002>, rot=<0.4879 -0.5118 -0.4879 -0.5118>}
 FACIAL_L_Ear {parent=FACIAL_C_FacialRoot, loc=<6.3528 -0.1327 0.9297>, rot=<-0.0288 -0.0006 -0.0262 -0.9992>}
 base_Earing_l {parent=FACIAL_L_Ear, loc=<0.0941 2.2725 -0.3861>, rot=<-0.0533 -0.0485 -0.6716 -0.7374>}
 dyn_Earing_l {parent=base_Earing_l, loc=<0 0 0>, rot=<0 0 0 1>}
 FACIAL_R_ForeheadIn {parent=FACIAL_C_FacialRoot, loc=<-1.6542 -2.5808 8.7846>, rot=<-0.0632 0.0696 -0.0041 -0.9956>}
 FACIAL_C_Forehead {parent=FACIAL_C_FacialRoot, loc=<-0.0008 -2.6819 8.887>, rot=<-0.0459 0 0 -0.9989>}
 FACIAL_L_ForeheadIn {parent=FACIAL_C_FacialRoot, loc=<1.6518 -2.5806 8.7864>, rot=<-0.0632 -0.0696 0.0041 -0.9956>}
 FACIAL_C_NoseBridge {parent=FACIAL_C_FacialRoot, loc=<-0.0003 -1.2349 9.6072>, rot=<0.0413 0 0 -0.9991>}
 FACIAL_C_Nose {parent=FACIAL_C_FacialRoot, loc=<-0.0006 0.7986 9.0471>, rot=<-0.1328 0 0 -0.9911>}
 FACIAL_R_CheekInner {parent=FACIAL_C_FacialRoot, loc=<-2.0449 1.4563 7.8417>, rot=<-0.1256 0.0875 0.1973 -0.9683>}
 FACIAL_R_EyesackLower {parent=FACIAL_C_FacialRoot, loc=<-3.3402 1.0827 7.199>, rot=<-0.0248 0.1396 0.0226 -0.9896>}
 FACIAL_R_CheekOuter {parent=FACIAL_C_FacialRoot, loc=<-3.8663 1.4733 6.2754>, rot=<0.0529 0.3772 -0.0994 -0.9193>}
 FACIAL_R_Masseter {parent=FACIAL_C_FacialRoot, loc=<-5.9161 1.9218 4.1734>, rot=<0.0669 -0.1129 -0.0082 -0.9913>}
 FACIAL_R_NasolabialFurrow {parent=FACIAL_C_FacialRoot, loc=<-2.7974 3.5 8.5108>, rot=<-0.0356 0.1249 0.3639 -0.9223>}
 FACIAL_R_CheekLower {parent=FACIAL_C_FacialRoot, loc=<-3.5949 4.2516 5.5559>, rot=<0.0323 0.427 -0.0567 -0.9019>}
 FACIAL_R_NasolabialBulge {parent=FACIAL_C_FacialRoot, loc=<-3.1152 4.5141 6.2269>, rot=<0.0556 0.3774 -0.0229 -0.9241>}
 FACIAL_L_CheekInner {parent=FACIAL_C_FacialRoot, loc=<2.0441 1.455 7.8413>, rot=<-0.1256 -0.0875 -0.1973 -0.9683>}
 FACIAL_L_EyesackLower {parent=FACIAL_C_FacialRoot, loc=<3.3401 1.0769 7.1981>, rot=<-0.0247 -0.1396 -0.0226 -0.9896>}
 FACIAL_L_CheekOuter {parent=FACIAL_C_FacialRoot, loc=<3.8667 1.474 6.2762>, rot=<0.0529 -0.3772 0.0994 -0.9193>}
 FACIAL_L_Masseter {parent=FACIAL_C_FacialRoot, loc=<5.9152 1.9184 4.176>, rot=<0.0669 0.1129 0.0082 -0.9913>}
 FACIAL_L_NasolabialFurrow {parent=FACIAL_C_FacialRoot, loc=<2.7944 3.5035 8.5113>, rot=<-0.0356 -0.1249 -0.3639 -0.9223>}
 FACIAL_L_CheekLower {parent=FACIAL_C_FacialRoot, loc=<3.5947 4.2544 5.5572>, rot=<0.0323 -0.427 0.0567 -0.9019>}
 FACIAL_L_NasolabialBulge {parent=FACIAL_C_FacialRoot, loc=<3.1147 4.5167 6.2283>, rot=<0.0556 -0.3774 0.0229 -0.9241>}
 FACIAL_L_Jawline {parent=FACIAL_C_FacialRoot, loc=<4.0464 5.4475 4.5754>, rot=<0.2069 0.1579 -0.0543 -0.964>}
 FACIAL_R_Jawline {parent=FACIAL_C_FacialRoot, loc=<-4.0479 5.4474 4.5726>, rot=<0.2069 -0.1579 0.0543 -0.964>}
 FACIAL_R_ForeheadMid {parent=FACIAL_C_FacialRoot, loc=<-3.1381 -2.8651 8.3005>, rot=<-0.0734 0.1756 0.021 -0.9815>}
 FACIAL_L_ForeheadMid {parent=FACIAL_C_FacialRoot, loc=<3.1348 -2.8653 8.3027>, rot=<-0.0734 -0.1756 -0.021 -0.9815>}
 FACIAL_R_Ear {parent=FACIAL_C_FacialRoot, loc=<-6.3488 -0.1323 0.9223>, rot=<-0.0288 0.0006 0.0262 -0.9992>}
 base_Earing_r {parent=FACIAL_R_Ear, loc=<-0.0383 2.2237 -0.3836>, rot=<0.0898 0.0794 -0.7346 -0.6679>}
 dyn_Earing_r {parent=base_Earing_r, loc=<0 0 0>, rot=<0 0 0 1>}
 head_null {parent=neck_02, type=Null, loc=<5.3914 0 0>, rot=<0 0 0.115 -0.9934>}
 clavicle_l {parent=spine_05, loc=<5.4831 -0.1053 -3.2946>, rot=<0.0465 -0.7289 -0.009 -0.6829>}
 upperarm_l {parent=clavicle_l, loc=<9.8658 -1.2136 -0.4105>, rot=<0.0164 -0.3245 -0.0269 -0.9454>}
 upperarm_twist_01_l {parent=upperarm_l, loc=<7.2135 -0.1134 0>, rot=<0 0 0 1>}
 clavicle_r {parent=spine_05, loc=<5.4831 -0.1053 3.2946>, rot=<0.7289 0.0465 -0.6829 0.009>}
 upperarm_r {parent=clavicle_r, loc=<-9.8658 1.2136 0.4105>, rot=<0.0164 -0.3245 -0.0269 -0.9454>}
 upperarm_twist_01_r {parent=upperarm_r, loc=<-7.2135 0.1134 0>, rot=<0 0 0 -1>}
 clavicle_l_null {parent=spine_05, type=Null, loc=<5.4831 -0.1053 -3.2946>, rot=<0.0465 -0.7289 -0.009 -0.6829>}
 clavicle_r_null {parent=spine_05, type=Null, loc=<5.4831 -0.1053 3.2946>, rot=<0.7289 0.0465 -0.6829 0.009>}
 neck_01_null {parent=spine_05, type=Null, loc=<11.3138 0.1529 0>, rot=<0 0 -0.2003 -0.9797>}
 head_null_sp {parent=spine_05, type=Null, loc=<20.6573 3.5489 0>, rot=<0 0 -0.036 -0.9994>}
 pelvis_ctrl {controlType=EULER_TRANSFORM, type=Control, loc=<0 0 0>, rot=<0 0 0 1>}
 spine_01_ctrl {controlType=EULER_TRANSFORM, parent=pelvis_ctrl, type=Control, loc=<0 0 0>, rot=<0 0 0 1>}
 spine_02_ctrl {controlType=EULER_TRANSFORM, parent=spine_01_ctrl, type=Control, loc=<0 0 0>, rot=<0 0 0 1>}
 spine_03_ctrl {controlType=EULER_TRANSFORM, parent=spine_02_ctrl, type=Control, loc=<0 0 0>, rot=<0 0 0 1>}
 spine_04_ctrl {controlType=EULER_TRANSFORM, parent=spine_03_ctrl, type=Control, loc=<0 0 0>, rot=<0 0 0 1>}
 spine_05_ctrl {controlType=EULER_TRANSFORM, parent=spine_04_ctrl, type=Control, loc=<0 0 0>, rot=<0 0 0 1>}
 neck_01_ctrl {controlType=EULER_TRANSFORM, parent=spine_05_ctrl, type=Control, loc=<0 0 0>, rot=<0 0 0 1>}
 neck_02_ctrl {controlType=EULER_TRANSFORM, parent=neck_01_ctrl, type=Control, loc=<0 0 0>, rot=<0 0 0 1>}
 head_ctrl {controlType=EULER_TRANSFORM, parent=neck_02_ctrl, type=Control, loc=<0 0 0>, rot=<0 0 0 1>}
 clavicle_l_ctrl {controlType=EULER_TRANSFORM, parent=spine_05_ctrl, type=Control, loc=<0 0 0>, rot=<0 0 0 1>}
 upperarm_l_ctrl {controlType=EULER_TRANSFORM, parent=clavicle_l_ctrl, type=Control, loc=<0 0 0>, rot=<0 0 0 1>}
 clavicle_r_ctrl {controlType=EULER_TRANSFORM, parent=spine_05_ctrl, type=Control, loc=<0 0 0>, rot=<0 0 0 1>}
 upperarm_r_ctrl {controlType=EULER_TRANSFORM, parent=clavicle_r_ctrl, type=Control, loc=<0 0 0>, rot=<0 0 0 1>}
 dyn_hair_L_01_null {type=Null, loc=<4.8248 -8.5 154.5584>, rot=<-0.4397 -0.3968 0.5474 -0.5913>}
 dyn_hair_L_01_ctrl {controlType=EULER_TRANSFORM, parent=dyn_hair_L_01_null, type=Control, loc=<-2.692 -8.3184 154.4532>, rot=<0.4491 0.3865 -0.5396 0.5981>}
 base_hair_L_null {type=Null, loc=<4.8248 -7.6896 156.5481>, rot=<0.4301 0.3865 -0.555 0.5981>}
 base_hair_L_ctrl {controlType=EULER_TRANSFORM, parent=base_hair_L_null, type=Control, loc=<0 0 0>, rot=<0 0 0 1>}
 dyn_hair_L_02_null {type=Null, loc=<5.3542 -9.532 150.8719>, rot=<-0.4397 -0.3968 0.5474 -0.5913>}
 dyn_hair_L_02_ctrl {controlType=EULER_TRANSFORM, parent=dyn_hair_L_02_null, type=Control, loc=<0 0 0>, rot=<0 0 0 1>}
 dyn_hair_L_03_null {type=Null, loc=<5.9778 -11.6637 143.96>, rot=<0.4643 0.4235 -0.5266 0.5724>}
 dyn_hair_L_03_ctrl {controlType=EULER_TRANSFORM, parent=dyn_hair_L_03_null, type=Control, loc=<0 0 0>, rot=<0 0 0 1>}
 dyn_hair_L_04_null {type=Null, loc=<6.796 -13.7263 134.768>, rot=<-0.4813 -0.442 0.5112 -0.5583>}
 dyn_hair_L_04_ctrl {controlType=EULER_TRANSFORM, parent=dyn_hair_L_04_null, type=Control, loc=<0 0 0>, rot=<0 0 0 1>}
 dyn_hair_L_05_null {type=Null, loc=<7.4972 -14.8934 126.7918>, rot=<0.5236 0.4886 -0.4677 0.5181>}
 dyn_hair_L_05_ctrl {controlType=EULER_TRANSFORM, parent=dyn_hair_L_05_null, type=Control, loc=<0 0 0>, rot=<0 0 0 1>}
 dyn_hair_L_06_null {type=Null, loc=<8.158 -14.7555 119.0693>, rot=<0.5236 0.4886 -0.4677 0.5181>}
 dyn_hair_L_06_ctrl {controlType=EULER_TRANSFORM, parent=dyn_hair_L_06_null, type=Control, loc=<0 0 0>, rot=<0 0 0 1>}
 dyn_hair_M_01_null {type=Null, loc=<-0.077 -8.5697 154.1377>, rot=<-0.4186 -0.4186 0.5699 -0.5699>}
 dyn_hair_M_01_ctrl {controlType=EULER_TRANSFORM, parent=dyn_hair_M_01_null, type=Control, loc=<-8.214 -8.4926 153.9842>, rot=<0.4285 0.4086 -0.5625 0.5771>}
 base_hair_M_null {type=Null, loc=<0 -7.6864 156.7647>, rot=<0.4086 0.4086 -0.5771 0.5771>}
 base_hair_M_ctrl {controlType=EULER_TRANSFORM, parent=base_hair_M_null, type=Control, loc=<0 0 0>, rot=<0 0 0 1>}
 dyn_hair_M_02_null {type=Null, loc=<0 -9.5309 150.8782>, rot=<-0.4186 -0.4186 0.5699 -0.5699>}
 dyn_hair_M_02_ctrl {controlType=EULER_TRANSFORM, parent=dyn_hair_M_02_null, type=Control, loc=<0 0 0>, rot=<0 0 0 1>}
 dyn_hair_M_03_null {type=Null, loc=<0 -11.6671 143.9396>, rot=<0.4443 0.4443 -0.5501 0.5501>}
 dyn_hair_M_03_ctrl {controlType=EULER_TRANSFORM, parent=dyn_hair_M_03_null, type=Control, loc=<0 0 0>, rot=<0 0 0 1>}
 dyn_hair_M_04_null {type=Null, loc=<0 -13.7355 134.7124>, rot=<-0.4621 -0.4621 0.5353 -0.5353>}
 dyn_hair_M_04_ctrl {controlType=EULER_TRANSFORM, parent=dyn_hair_M_04_null, type=Control, loc=<0 0 0>, rot=<0 0 0 1>}
 dyn_hair_M_05_null {type=Null, loc=<0 -14.9076 126.7063>, rot=<0.5066 0.5066 -0.4934 0.4934>}
 dyn_hair_M_05_ctrl {controlType=EULER_TRANSFORM, parent=dyn_hair_M_05_null, type=Control, loc=<0 0 0>, rot=<0 0 0 1>}
 dyn_hair_M_06_null {type=Null, loc=<0 -14.7743 118.9555>, rot=<0.5066 0.5066 -0.4934 0.4934>}
 dyn_hair_M_06_ctrl {controlType=EULER_TRANSFORM, parent=dyn_hair_M_06_null, type=Control, loc=<0 0 0>, rot=<0 0 0 1>}
 dyn_hair_R_01_null {type=Null, loc=<-5.0436 -8.5697 154.1377>, rot=<-0.3968 -0.4397 0.5913 -0.5474>}
 dyn_hair_R_01_ctrl {controlType=EULER_TRANSFORM, parent=dyn_hair_R_01_null, type=Control, loc=<-13.1572 -8.5863 153.8084>, rot=<0.407 0.4301 -0.5843 0.555>}
 base_hair_R_null {type=Null, loc=<-4.8248 -7.6912 156.7358>, rot=<0.3865 0.4301 -0.5981 0.555>}
 base_hair_R_ctrl {controlType=EULER_TRANSFORM, parent=base_hair_R_null, type=Control, loc=<0 0 0>, rot=<0 0 0 1>}
 dyn_hair_R_02_null {type=Null, loc=<-5.3542 -9.532 150.8719>, rot=<-0.3968 -0.4397 0.5913 -0.5474>}
 dyn_hair_R_02_ctrl {controlType=EULER_TRANSFORM, parent=dyn_hair_R_02_null, type=Control, loc=<0 0 0>, rot=<0 0 0 1>}
 dyn_hair_R_03_null {type=Null, loc=<-5.9778 -11.6637 143.96>, rot=<0.4235 0.4643 -0.5724 0.5266>}
 dyn_hair_R_03_ctrl {controlType=EULER_TRANSFORM, parent=dyn_hair_R_03_null, type=Control, loc=<0 0 0>, rot=<0 0 0 1>}
 dyn_hair_R_04_null {type=Null, loc=<-6.796 -13.7263 134.768>, rot=<-0.442 -0.4813 0.5583 -0.5112>}
 dyn_hair_R_04_ctrl {controlType=EULER_TRANSFORM, parent=dyn_hair_R_04_null, type=Control, loc=<0 0 0>, rot=<0 0 0 1>}
 dyn_hair_R_05_null {type=Null, loc=<-7.4972 -14.8934 126.7918>, rot=<0.4886 0.5236 -0.5181 0.4677>}
 dyn_hair_R_05_ctrl {controlType=EULER_TRANSFORM, parent=dyn_hair_R_05_null, type=Control, loc=<0 0 0>, rot=<0 0 0 1>}
 dyn_hair_R_06_null {type=Null, loc=<-8.158 -14.7555 119.0693>, rot=<0.4886 0.5236 -0.5181 0.4677>}
 dyn_hair_R_06_ctrl {controlType=EULER_TRANSFORM, parent=dyn_hair_R_06_null, type=Control, loc=<0 0 0>, rot=<0 0 0 1>}
]
variables [
 ClippingCorrect {type=bool, default=False}
 IsPlayCelebrateEmote {type=double, default=0}
 dyn_hair_l_controls {type="TArray<FRigElementKey>"}
 dyn_hair_l_pose {type=FRigPose, default=(Elements=,HierarchyTopologyVersion=-1,PoseHash=-1)}
 dyn_hair_m_controls {type="TArray<FRigElementKey>"}
 dyn_hair_m_pose {type=FRigPose, default=(Elements=,HierarchyTopologyVersion=-1,PoseHash=-1)}
 dyn_hair_r_controls {type="TArray<FRigElementKey>", default=(())}
 dyn_hair_r_pose {type=FRigPose, default=(Elements=,HierarchyTopologyVersion=-1,PoseHash=-1)}
 isCrouching {type=bool, default=False}
 isDBNO {type=bool, default=False}
 isDivingNoBackBling {type=bool, default=False}
 isHeadUp {type=double, default=0}
 isPistolEquiped {type=bool, default=False}
 isWearingBackBling {type=bool, default=False}
 ismoving2D {type=bool, default=False}
 pub aimPitch {type=double, default=0}
 pub disableUpperLimbsCorrection {type=bool, default=False}
 pub disablecontrolRig {type=bool, default=False}
 pub diveFloating {type=bool, default=False}
 pub isAssualtedRifleEquipped {type=double, default=0}
 pub isDiving {type=bool, default=False}
 pub isFloating {type=bool, default=False}
 pub isTargeting {type=bool, default=False}
]
