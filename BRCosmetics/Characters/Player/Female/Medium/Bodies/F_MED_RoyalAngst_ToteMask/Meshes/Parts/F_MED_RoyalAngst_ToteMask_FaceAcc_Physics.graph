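# Physics asset for the face accessory of the female medium RoyalAngst ToteMask cosmetic, with three capsule chains (left, middle, right) on dyn_hair bones.
# Each chain roots on spine_05 with tight swing and locked twist, then has per-joint limits that widen toward the tip on swing1 and narrow on swing2.
# Free-angle constraints tie the middle tip to both side chains.
# Pelvis, spine, clavicle and upper-arm capsules form a body proxy joined by free-rotation constraints.
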PhysicsAsset /BRCosmetics/Characters/Player/Female/Medium/Bodies/F_MED_RoyalAngst_ToteMask/Meshes/Parts/F_MED_RoyalAngst_ToteMask_FaceAcc_Physics
bodies {
 clavicle_l [{shapeType=Capsule, length=7.5725, radius=5, shapeName=clavicle_l_capsule, center=<4 0 0>, rotation=<rot 90 0 0>}]
 clavicle_r [{shapeType=Capsule, length=7.5725, radius=5, shapeName=clavicle_r_capsule, center=<-4 0 0>, rotation=<rot 90 0 0>}]
 dyn_hair_L_02 [{shapeType=Capsule, length=2, radius=1, shapeName=dyn_hair_L_02_capsule}]
 dyn_hair_L_03 [{
   shapeType=Capsule
   length=3.4992
   radius=3.5008
   shapeName=dyn_hair_L_04_capsule
   center=<2.3707 0.9173 0.0223>
   rotation=<rot 69.5602 -179.9999 -179.9999>
  }]
 dyn_hair_L_04 [{
   shapeType=Capsule
   length=3.4992
   radius=3.5008
   shapeName=dyn_hair_L_04_capsule
   center=<2.0103 0.9173 0.9894>
   rotation=<rot 88.2928 -179.9996 -179.9996>
  }]
 dyn_hair_L_05 [{
   shapeType=Capsule
   length=3.6071
   radius=4.0234
   shapeName=dyn_hair_L_05_capsule
   center=<2.7326 1.07 0.8007>
   rotation=<rot 88.6414 -179.9993 176.46>
  }]
 dyn_hair_L_06 [{
   shapeType=Capsule
   length=2.3027
   radius=4.8252
   shapeName=dyn_hair_L_06_capsule
   center=<3.0963 2.0138 0.5151>
   rotation=<rot 82.0882 -0.0001 -6.8512>
  }]
 dyn_hair_M_02 [{shapeType=Capsule, length=2, radius=1, shapeName=dyn_hair_M_02_capsule}]
 dyn_hair_M_03 [{shapeType=Capsule, length=3.4992, radius=3.5008, shapeName=dyn_hair_M_04_capsule, center=<2.3707 0 0.0223>, rotation=<rot 69.5602 180 180>}]
 dyn_hair_M_04 [{shapeType=Capsule, length=3.4992, radius=3.5008, shapeName=dyn_hair_M_04_capsule, center=<2.0103 0 0.9894>, rotation=<rot 88.2928 -180 -180>}]
 dyn_hair_M_05 [{shapeType=Capsule, length=3.7583, radius=3.8722, shapeName=dyn_hair_M_05_capsule, center=<2.7326 0 0.8007>, rotation=<rot 88.6413 -180 -180>}]
 dyn_hair_M_06 [{shapeType=Capsule, length=2.3027, radius=4.8252, shapeName=dyn_hair_M_06_capsule, center=<3.1323 0 1.1009>, rotation=<rot 82.0882 0 0>}]
 dyn_hair_R_02 [{shapeType=Capsule, length=2, radius=1, shapeName=dyn_hair_R_02_capsule}]
 dyn_hair_R_03 [{
   shapeType=Capsule
   length=3.4992
   radius=3.5008
   shapeName=dyn_hair_R_04_capsule
   center=<2.3706 -1.5457 0.0223>
   rotation=<rot 69.5609 178.5679 -179.8673>
  }]
 dyn_hair_R_04 [{
   shapeType=Capsule
   length=3.4992
   radius=3.5008
   shapeName=dyn_hair_R_04_capsule
   center=<2.0103 -1.5367 0.9894>
   rotation=<rot 88.2928 179.9995 -178.4434>
  }]
 dyn_hair_R_05 [{
   shapeType=Capsule
   length=3.7583
   radius=3.8722
   shapeName=dyn_hair_R_05_capsule
   center=<2.5947 -1.8892 1.2053>
   rotation=<rot 89.5086 -11.2819 -8.6398>
  }]
 dyn_hair_R_06 [{
   shapeType=Capsule
   length=2.3027
   radius=4.8252
   shapeName=dyn_hair_R_06_capsule
   center=<3.1094 -1.7501 0.7525>
   rotation=<rot 82.088 0.0002 1.8524>
  }]
 pelvis [{shapeType=Capsule, length=6.6871, radius=12.2436, shapeName=pelvis_capsule}]
 spine_01 [{shapeType=Capsule, length=8.5167, radius=7.7843, shapeName=spine_01_capsule, center=<1.6844 0 0>}]
 spine_02 [{shapeType=Capsule, length=4.2488, radius=7.922, shapeName=spine_02_capsule, center=<3.379 0.9679 0>}]
 spine_03 [{shapeType=Capsule, length=4.7259, radius=7.4449, shapeName=spine_03_capsule, center=<4.0919 0.0579 0>}]
 spine_04 [{shapeType=Capsule, length=6.8498, radius=7.8608, shapeName=spine_04_capsule, center=<3.9628 -0.6579 0>}]
 spine_05 [{shapeType=Capsule, length=10.1632, radius=8.6265, shapeName=spine_05_capsule, center=<2.5194 0.7308 0>}]
 upperarm_twist_01_l [{shapeType=Capsule, length=29.689, radius=6.902, shapeName=upperarm_l_capsule, center=<13.1892 0 0>, rotation=<rot 90 0 0>}]
 upperarm_twist_01_r [{shapeType=Capsule, length=29.689, radius=6.902, shapeName=upperarm_r_capsule, center=<-13.0017 0 0>, rotation=<rot 90 0 0>}]
}
constraints [
 {
  bone1Name=clavicle_l
  bone2Name=spine_05
  swing1LimitDegrees=45
  swing1Motion=Free
  swing2LimitDegrees=45
  swing2Motion=Free
  twistLimitDegrees=45
  twistMotion=Free
 }
 {
  bone1Name=clavicle_r
  bone2Name=spine_05
  swing1LimitDegrees=45
  swing1Motion=Free
  swing2LimitDegrees=45
  swing2Motion=Free
  twistLimitDegrees=45
  twistMotion=Free
 }
 {
  bone1Name=spine_05
  bone2Name=spine_04
  swing1LimitDegrees=45
  swing1Motion=Free
  swing2LimitDegrees=45
  swing2Motion=Free
  twistLimitDegrees=45
  twistMotion=Free
 }
 {
  bone1Name=spine_04
  bone2Name=spine_03
  swing1LimitDegrees=45
  swing1Motion=Free
  swing2LimitDegrees=45
  swing2Motion=Free
  twistLimitDegrees=45
  twistMotion=Free
 }
 {
  bone1Name=spine_03
  bone2Name=spine_02
  swing1LimitDegrees=45
  swing1Motion=Free
  swing2LimitDegrees=45
  swing2Motion=Free
  twistLimitDegrees=45
  twistMotion=Free
 }
 {
  bone1Name=spine_02
  bone2Name=spine_01
  swing1LimitDegrees=45
  swing1Motion=Free
  swing2LimitDegrees=45
  swing2Motion=Free
  twistLimitDegrees=45
  twistMotion=Free
 }
 {
  bone1Name=spine_01
  bone2Name=pelvis
  swing1LimitDegrees=45
  swing1Motion=Free
  swing2LimitDegrees=45
  swing2Motion=Free
  twistLimitDegrees=45
  twistMotion=Free
 }
 {bone1Name=dyn_hair_L_06, bone2Name=dyn_hair_L_05, swing1LimitDegrees=15, swing2LimitDegrees=20, twistLimitDegrees=0}
 {bone1Name=dyn_hair_L_05, bone2Name=dyn_hair_L_04, swing1LimitDegrees=12, swing2LimitDegrees=20, twistLimitDegrees=0}
 {bone1Name=dyn_hair_L_04, bone2Name=dyn_hair_L_03, swing1LimitDegrees=9, swing2LimitDegrees=30, twistLimitDegrees=0}
 {bone1Name=dyn_hair_L_03, bone2Name=dyn_hair_L_02, swing1LimitDegrees=6, swing2LimitDegrees=35, twistLimitDegrees=0}
 {bone1Name=dyn_hair_M_06, bone2Name=dyn_hair_M_05, swing1LimitDegrees=15, swing2LimitDegrees=20, twistLimitDegrees=0}
 {bone1Name=dyn_hair_M_05, bone2Name=dyn_hair_M_04, swing1LimitDegrees=12, swing2LimitDegrees=20, twistLimitDegrees=0}
 {bone1Name=dyn_hair_M_04, bone2Name=dyn_hair_M_03, swing1LimitDegrees=9, swing2LimitDegrees=30, twistLimitDegrees=0}
 {bone1Name=dyn_hair_M_03, bone2Name=dyn_hair_M_02, swing1LimitDegrees=6, swing2LimitDegrees=35, twistLimitDegrees=0}
 {bone1Name=dyn_hair_R_06, bone2Name=dyn_hair_R_05, swing1LimitDegrees=15, swing2LimitDegrees=20, twistLimitDegrees=0}
 {bone1Name=dyn_hair_R_05, bone2Name=dyn_hair_R_04, swing1LimitDegrees=12, swing2LimitDegrees=20, twistLimitDegrees=0}
 {bone1Name=dyn_hair_R_04, bone2Name=dyn_hair_R_03, swing1LimitDegrees=9, swing2LimitDegrees=30, twistLimitDegrees=0}
 {bone1Name=dyn_hair_R_03, bone2Name=dyn_hair_R_02, swing1LimitDegrees=6, swing2LimitDegrees=35, twistLimitDegrees=0}
 {
  bone1Name=upperarm_twist_01_r
  bone2Name=clavicle_r
  swing1LimitDegrees=45
  swing1Motion=Free
  swing2LimitDegrees=45
  swing2Motion=Free
  twistLimitDegrees=45
  twistMotion=Free
 }
 {
  bone1Name=upperarm_twist_01_l
  bone2Name=clavicle_l
  swing1LimitDegrees=45
  swing1Motion=Free
  swing2LimitDegrees=45
  swing2Motion=Free
  twistLimitDegrees=45
  twistMotion=Free
 }
 {
  bone1Name=dyn_hair_M_06
  bone2Name=dyn_hair_L_05
  swing1LimitDegrees=45
  swing1Motion=Free
  swing2LimitDegrees=45
  swing2Motion=Free
  twistLimitDegrees=45
  twistMotion=Free
 }
 {
  bone1Name=dyn_hair_M_06
  bone2Name=dyn_hair_R_05
  swing1LimitDegrees=45
  swing1Motion=Free
  swing2LimitDegrees=45
  swing2Motion=Free
  twistLimitDegrees=45
  twistMotion=Free
 }
 {bone1Name=dyn_hair_L_02, bone2Name=spine_05, swing1LimitDegrees=10, swing2LimitDegrees=10, twistLimitDegrees=45, twistMotion=Locked}
 {bone1Name=dyn_hair_M_02, bone2Name=spine_05, swing1LimitDegrees=10, swing2LimitDegrees=10, twistLimitDegrees=45, twistMotion=Locked}
 {bone1Name=dyn_hair_R_02, bone2Name=spine_05, swing1LimitDegrees=10, swing2LimitDegrees=10, twistLimitDegrees=45, twistMotion=Locked}
]
