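# Animation Blueprint for the RoyalAngst ToteMask female face-accessory mesh, animating hair and earrings.
# Copied body and face poses feed a Control Rig, then AnimDynamics for front hair and left earring.
# A further AnimDynamics, bone-offset and RigidBody hair pass is bool-blended with that pose at output; the cached BackblingHair pass is unused there.
# The event graph derives gravity, spring stiffness, physics alpha and state flags from movement, weapon, emote and skydiving state.
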AnimBlueprint /BRCosmetics/Characters/Player/Female/Medium/Bodies/F_MED_RoyalAngst_ToteMask/Meshes/Parts/F_MED_RoyalAngst_ToteMask_FaceAcc_AnimBP
target_skeleton=&/F_MED_RoyalAngst_ToteMask_FaceAcc_Skeleton
graphs [
 AnimGraph {nodes [
   n1 = Blend Poses by bool(BlendPose_0=Use cached pose 'NoBackblingHair'(), BlendPose_1=Use cached pose 'Front2backhair'()) {blendType=Linear}
   n2 = Use cached pose 'Front2backhair'()
   Output Pose(Result=Blend Poses by bool(BlendPose_0=Use cached pose 'Front2backhair'(), BlendPose_1=Blend Poses by bool(BlendPose_0=@n2, BlendPose_1=@n1) {blendType=Linear}) {blendType=Linear}) {layerGroup=None}
   n3 = Control Rig!(
    Source=Use cached pose 'Start_Cache'()
    Alpha
    pub disablecontrolRig:bool=Get disableControlRig() -> (disableControlRig:bool)
    pub diveFloating:bool=Get bIsSkydiveFloating()
    pub disableUpperLimbsCorrection:bool=Get disableUpperLimbsCorrection() -> (disableUpperLimbsCorrection:bool)
    pub isTargeting:bool=Get bIsTargeting()
    pub aimPitch:double=Get AimPitch()
    pub isAssualtedRifleEquipped:double=Get isAssaultRifleEquipped float() -> (isAssaultRifleEquipped float:double)
    isWearingBackBling:bool=Get Is Backbling Equipped() -> (Is Backbling Equipped:bool)
    isHeadUp:double=Get Head Up() -> (Head Up:float)
    pub isDiving:bool=Get bIsSkydiveDiveMode()
    isDivingNoBackBling:bool=Get DiveModeNoBB() -> (DiveModeNoBB:bool)
    isCrouching:bool=Get bIsCrouching()
    isPistolEquiped:bool=Get isPistolEquipped() -> (isPistolEquipped:bool)
    ismoving2D:bool=Get bIsMoving2D()
    IsPlayCelebrateEmote:double=Get IsPlayCelebrateEmote() -> (IsPlayCelebrateEmote:double)
    isDBNO:bool=Get bIsDBNO()
   ) -> (Pose) {
    alphaBoolBlend {blendOption=Linear}
    controlRigAssetReference {blueprintRigClass=&/F_MED_RoyalAngst_ToteMask_FaceAcc_CtrlRig.F_MED_RoyalAngst_ToteMask_FaceAcc_CtrlRig_C}
   }
   Save cached pose 'CR2AnimDynamics'(Pose=@n3) {cacheName=CR2AnimDynamics}
   n4 = Anim Dynamics - Bone: dyn_frontHair_05!(
    ComponentPose=Local To Component(LocalPose=Use cached pose 'CR2AnimDynamics'())
    Alpha=1
    GravityOverride:Vector="0,0,-980"
    ExternalForce="0,0,0"
   ) -> (Pose) {
    angularDampingOverride=1
    angularSpringConstant=20000
    bAngularSpring=true
    bOverrideAngularDamping=true
    bUseGravityOverride=true
    gravityScale=0.35
    simulationSpace=RootRelative
    alphaBoolBlend {blendOption=Linear}
    boundBone {boneName=dyn_frontHair_05}
    componentLinearAccScale=<1 1 1>
    componentLinearVelScale=<1 1 1>
    gravityOverride {z=-980}
    physicsBodyDefinitions [{
      boundBone {boneName=dyn_frontHair_05}
      boxExtents=<20 30 30>
      constraintSetup {angularLimitsMax=<15 2 2>, angularLimitsMin=<-15 -15 -15>, angularTarget=<1 0 0>}
      localJointOffset=<10 0 0>
     }]
   }
   n5 = Anim Dynamics - Bone: dyn_Earing_l!(ComponentPose=@n4, Alpha=1, GravityOverride:Vector="0,0,-980", ExternalForce="0,0,0") -> (Pose) {
    angularDampingOverride=1
    angularSpringConstant=20000
    bAngularSpring=true
    bOverrideAngularDamping=true
    bUseGravityOverride=true
    gravityScale=0.35
    simulationSpace=RootRelative
    alphaBoolBlend {blendOption=Linear}
    boundBone {boneName=dyn_Earing_l}
    componentLinearAccScale=<1 1 1>
    componentLinearVelScale=<1 1 1>
    gravityOverride {z=-980}
    physicsBodyDefinitions [{
      boundBone {boneName=dyn_Earing_l}
      boxExtents=<20 30 30>
      constraintSetup {angularLimitsMax=<15 15 15>, angularLimitsMin=<-15 -15 -15>, angularTarget=<1 0 0>}
      localJointOffset=<4 0 0>
     }]
   }
   Save cached pose 'Front2backhair'(Pose=Component To Local(ComponentPose=@n5)) {cacheName=Front2backhair}
   n6 = Anim Dynamics - Bone: dyn_hair_L_01!(
    ComponentPose=Local To Component(LocalPose=Use cached pose 'Front2backhair'())
    bAlphaBoolEnabled:bool=True
    GravityOverride:Vector="0,0,-980"
    AngularSpringConstant:float=20000
    ExternalForce="0,0,0"
   ) -> (Pose) {
    alphaInputType=Bool
    angularDampingOverride=1
    angularSpringConstant=20000
    bAngularSpring=true
    bOverrideAngularDamping=true
    bUseGravityOverride=true
    gravityScale=0.35
    simulationSpace=RootRelative
    alphaBoolBlend {blendInTime=2, blendOption=Cubic}
    boundBone {boneName=dyn_hair_L_01}
    componentAppliedLinearAccClamp=<3000 3000 3000>
    componentLinearAccScale=<0.3 0.4 0.3>
    componentLinearVelScale=<0.3 0.4 0.3>
    gravityOverride {z=-980}
    physicsBodyDefinitions [{
      boundBone {boneName=dyn_hair_L_01}
      boxExtents=<25 25 25>
      constraintSetup {angularLimitsMax=<0 0 10>, angularLimitsMin=<0 -40 -10>, angularTarget=<1 0 0>}
      localJointOffset=<2 0 0>
     }]
   }
   n7 = Anim Dynamics - Bone: dyn_hair_M_01!(ComponentPose=@n6, bAlphaBoolEnabled:bool=True, GravityOverride:Vector="0,0,-980", AngularSpringConstant:float=20000, ExternalForce="0,0,0") -> (Pose) {
    alphaInputType=Bool
    angularDampingOverride=1
    angularSpringConstant=20000
    bAngularSpring=true
    bOverrideAngularDamping=true
    bUseGravityOverride=true
    gravityScale=0.35
    simulationSpace=RootRelative
    alphaBoolBlend {blendInTime=2, blendOption=Cubic}
    boundBone {boneName=dyn_hair_M_01}
    componentAppliedLinearAccClamp=<3000 3000 3000>
    componentLinearAccScale=<0.3 0.4 0.3>
    componentLinearVelScale=<0.3 0.4 0.3>
    gravityOverride {z=-980}
    physicsBodyDefinitions [{
      boundBone {boneName=dyn_hair_M_01}
      boxExtents=<25 25 25>
      constraintSetup {angularLimitsMax=<0 0 10>, angularLimitsMin=<0 -40 -10>, angularTarget=<1 0 0>}
      localJointOffset=<2 0 0>
     }]
   }
   n8 = Anim Dynamics - Bone: dyn_hair_R_01!(ComponentPose=@n7, bAlphaBoolEnabled:bool=True, GravityOverride:Vector="0,0,-980", AngularSpringConstant:float=20000, ExternalForce="0,0,0") -> (Pose) {
    alphaInputType=Bool
    angularDampingOverride=1
    angularSpringConstant=20000
    bAngularSpring=true
    bOverrideAngularDamping=true
    bUseGravityOverride=true
    gravityScale=0.35
    simulationSpace=RootRelative
    alphaBoolBlend {blendInTime=2, blendOption=Cubic}
    boundBone {boneName=dyn_hair_R_01}
    componentAppliedLinearAccClamp=<3000 3000 3000>
    componentLinearAccScale=<0.3 0.4 0.3>
    componentLinearVelScale=<0.3 0.4 0.3>
    gravityOverride {z=-980}
    physicsBodyDefinitions [{
      boundBone {boneName=dyn_hair_R_01}
      boxExtents=<25 25 25>
      constraintSetup {angularLimitsMax=<0 0 10>, angularLimitsMin=<0 -40 -10>, angularTarget=<1 0 0>}
      localJointOffset=<2 0 0>
     }]
   }
   n9 = Anim Dynamics - Bone: dyn_hair_L_02!(ComponentPose=@n8, bAlphaBoolEnabled:bool=True, GravityOverride:Vector="0,0,-980", AngularSpringConstant:float=20000, ExternalForce="0,0,0") -> (Pose) {
    alphaInputType=Bool
    angularDampingOverride=1
    angularSpringConstant=20000
    bAngularSpring=true
    bOverrideAngularDamping=true
    bUseGravityOverride=true
    gravityScale=0.35
    simulationSpace=RootRelative
    alphaBoolBlend {blendInTime=2, blendOption=Cubic}
    boundBone {boneName=dyn_hair_L_02}
    componentAppliedLinearAccClamp=<3000 3000 3000>
    componentLinearAccScale=<0.3 0.4 0.3>
    componentLinearVelScale=<0.3 0.4 0.3>
    gravityOverride {z=-980}
    physicsBodyDefinitions [{
      boundBone {boneName=dyn_hair_L_02}
      boxExtents=<25 25 25>
      constraintSetup {angularLimitsMax=<0 0 7>, angularLimitsMin=<0 -35 -7>, angularTarget=<1 0 0>}
      localJointOffset=<2 0 0>
     }]
   }
   n10 = Anim Dynamics - Bone: dyn_hair_M_02!(ComponentPose=@n9, bAlphaBoolEnabled:bool=True, GravityOverride:Vector="0,0,-980", AngularSpringConstant:float=20000, ExternalForce="0,0,0") -> (Pose) {
    alphaInputType=Bool
    angularDampingOverride=1
    angularSpringConstant=20000
    bAngularSpring=true
    bOverrideAngularDamping=true
    bUseGravityOverride=true
    gravityScale=0.35
    simulationSpace=RootRelative
    alphaBoolBlend {blendInTime=2, blendOption=Cubic}
    boundBone {boneName=dyn_hair_M_02}
    componentAppliedLinearAccClamp=<3000 3000 3000>
    componentLinearAccScale=<0.3 0.4 0.3>
    componentLinearVelScale=<0.3 0.4 0.3>
    gravityOverride {z=-980}
    physicsBodyDefinitions [{
      boundBone {boneName=dyn_hair_M_02}
      boxExtents=<25 25 25>
      constraintSetup {angularLimitsMax=<0 0 7>, angularLimitsMin=<0 -35 -7>, angularTarget=<1 0 0>}
      localJointOffset=<2 0 0>
     }]
   }
   n11 = Anim Dynamics - Bone: dyn_hair_R_02!(ComponentPose=@n10, bAlphaBoolEnabled:bool=True, GravityOverride:Vector="0,0,-980", AngularSpringConstant:float=20000, ExternalForce="0,0,0") -> (Pose) {
    alphaInputType=Bool
    angularDampingOverride=1
    angularSpringConstant=20000
    bAngularSpring=true
    bOverrideAngularDamping=true
    bUseGravityOverride=true
    gravityScale=0.35
    simulationSpace=RootRelative
    alphaBoolBlend {blendInTime=2, blendOption=Cubic}
    boundBone {boneName=dyn_hair_R_02}
    componentAppliedLinearAccClamp=<3000 3000 3000>
    componentLinearAccScale=<0.3 0.4 0.3>
    componentLinearVelScale=<0.3 0.4 0.3>
    gravityOverride {z=-980}
    physicsBodyDefinitions [{
      boundBone {boneName=dyn_hair_R_02}
      boxExtents=<25 25 25>
      constraintSetup {angularLimitsMax=<0 0 7>, angularLimitsMin=<0 -35 -7>, angularTarget=<1 0 0>}
      localJointOffset=<2 0 0>
     }]
   }
   n12 = Anim Dynamics - Bone: dyn_hair_L_03!(ComponentPose=@n11, bAlphaBoolEnabled:bool=True, GravityOverride:Vector="0,0,-980", AngularSpringConstant:float=20000, ExternalForce="0,0,0") -> (Pose) {
    alphaInputType=Bool
    angularDampingOverride=1
    angularSpringConstant=20000
    bAngularSpring=true
    bOverrideAngularDamping=true
    bUseGravityOverride=true
    gravityScale=0.35
    simulationSpace=RootRelative
    alphaBoolBlend {blendInTime=2, blendOption=Cubic}
    boundBone {boneName=dyn_hair_L_03}
    componentAppliedLinearAccClamp=<3000 3000 3000>
    componentLinearAccScale=<0.3 0.4 0.3>
    componentLinearVelScale=<0.3 0.4 0.3>
    gravityOverride {z=-980}
    physicsBodyDefinitions [{
      boundBone {boneName=dyn_hair_L_03}
      boxExtents=<25 25 25>
      constraintSetup {angularLimitsMax=<0 0 5>, angularLimitsMin=<0 -15 -5>, angularTarget=<1 0 0>}
      localJointOffset=<2 0 0>
     }]
   }
   n13 = Anim Dynamics - Bone: dyn_hair_M_03!(ComponentPose=@n12, bAlphaBoolEnabled:bool=True, GravityOverride:Vector="0,0,-980", AngularSpringConstant:float=20000, ExternalForce="0,0,0") -> (Pose) {
    alphaInputType=Bool
    angularDampingOverride=1
    angularSpringConstant=20000
    bAngularSpring=true
    bOverrideAngularDamping=true
    bUseGravityOverride=true
    gravityScale=0.35
    simulationSpace=RootRelative
    alphaBoolBlend {blendInTime=2, blendOption=Cubic}
    boundBone {boneName=dyn_hair_M_03}
    componentAppliedLinearAccClamp=<3000 3000 3000>
    componentLinearAccScale=<0.3 0.4 0.3>
    componentLinearVelScale=<0.3 0.4 0.3>
    gravityOverride {z=-980}
    physicsBodyDefinitions [{
      boundBone {boneName=dyn_hair_M_03}
      boxExtents=<25 25 25>
      constraintSetup {angularLimitsMax=<0 0 5>, angularLimitsMin=<0 -15 -5>, angularTarget=<1 0 0>}
      localJointOffset=<2 0 0>
     }]
   }
   n14 = Anim Dynamics - Bone: dyn_hair_R_03!(ComponentPose=@n13, bAlphaBoolEnabled:bool=True, GravityOverride:Vector="0,0,-980", AngularSpringConstant:float=20000, ExternalForce="0,0,0") -> (Pose) {
    alphaInputType=Bool
    angularDampingOverride=1
    angularSpringConstant=20000
    bAngularSpring=true
    bOverrideAngularDamping=true
    bUseGravityOverride=true
    gravityScale=0.35
    simulationSpace=RootRelative
    alphaBoolBlend {blendInTime=2, blendOption=Cubic}
    boundBone {boneName=dyn_hair_R_03}
    componentAppliedLinearAccClamp=<3000 3000 3000>
    componentLinearAccScale=<0.3 0.4 0.3>
    componentLinearVelScale=<0.3 0.4 0.3>
    gravityOverride {z=-980}
    physicsBodyDefinitions [{
      boundBone {boneName=dyn_hair_R_03}
      boxExtents=<25 25 25>
      constraintSetup {angularLimitsMax=<0 0 5>, angularLimitsMin=<0 -15 -5>, angularTarget=<1 0 0>}
      localJointOffset=<2 0 0>
     }]
   }
   n15 = "Transform (Modify) Bone - Bone: dyn_hair_L_07"!(ComponentPose=@n14, bAlphaBoolEnabled:bool=True, Translation="-6.257,-4.136,10.195", Rotation="-0.000,-0.000,-84.273") -> (Pose) {
    alphaInputType=Bool
    rotationMode=BMM_Additive
    rotationSpace=BCS_BoneSpace
    translationMode=BMM_Additive
    translationSpace=BCS_BoneSpace
    alphaBoolBlend {blendOption=QuadraticInOut, blendOutTime=2}
    boneToModify {boneName=dyn_hair_L_07}
    rotation=<rot 0 0 -84.273>
    translation=<-6.257 -4.136 10.195>
   }
   n16 = Anim Dynamics - Bone: dyn_hair_L_04!(ComponentPose=@n15, bAlphaBoolEnabled:bool=True, GravityOverride:Vector="0,0,-980", AngularSpringConstant:float=20000, ExternalForce="0,0,0") -> (Pose) {
    alphaInputType=Bool
    angularDampingOverride=1
    angularSpringConstant=20000
    bAngularSpring=true
    bOverrideAngularDamping=true
    bUseGravityOverride=true
    gravityScale=0.35
    simulationSpace=RootRelative
    alphaBoolBlend {blendInTime=2, blendOption=Cubic}
    boundBone {boneName=dyn_hair_L_04}
    componentAppliedLinearAccClamp=<3000 3000 3000>
    componentLinearAccScale=<0.3 0.4 0.3>
    componentLinearVelScale=<0.3 0.4 0.3>
    gravityOverride {z=-980}
    physicsBodyDefinitions [{
      boundBone {boneName=dyn_hair_L_04}
      boxExtents=<25 25 25>
      constraintSetup {angularLimitsMax=<0 20 5>, angularLimitsMin=<0 -10 -5>, angularTarget=<1 0 0>}
      localJointOffset=<2 0 0>
     }]
   }
   n17 = Anim Dynamics - Bone: dyn_hair_M_04!(ComponentPose=@n16, bAlphaBoolEnabled:bool=True, GravityOverride:Vector="0,0,-980", AngularSpringConstant:float=20000, ExternalForce="0,0,0") -> (Pose) {
    alphaInputType=Bool
    angularDampingOverride=1
    angularSpringConstant=20000
    bAngularSpring=true
    bOverrideAngularDamping=true
    bUseGravityOverride=true
    gravityScale=0.35
    simulationSpace=RootRelative
    alphaBoolBlend {blendInTime=2, blendOption=Cubic}
    boundBone {boneName=dyn_hair_M_04}
    componentAppliedLinearAccClamp=<3000 3000 3000>
    componentLinearAccScale=<0.3 0.4 0.3>
    componentLinearVelScale=<0.3 0.4 0.3>
    gravityOverride {z=-980}
    physicsBodyDefinitions [{
      boundBone {boneName=dyn_hair_M_04}
      boxExtents=<25 25 25>
      constraintSetup {angularLimitsMax=<0 20 5>, angularLimitsMin=<0 -10 -5>, angularTarget=<1 0 0>}
      localJointOffset=<2 0 0>
     }]
   }
   n18 = Anim Dynamics - Bone: dyn_hair_R_04!(ComponentPose=@n17, bAlphaBoolEnabled:bool=True, GravityOverride:Vector="0,0,-980", AngularSpringConstant:float=20000, ExternalForce="0,0,0") -> (Pose) {
    alphaInputType=Bool
    angularDampingOverride=1
    angularSpringConstant=20000
    bAngularSpring=true
    bOverrideAngularDamping=true
    bUseGravityOverride=true
    gravityScale=0.35
    simulationSpace=RootRelative
    alphaBoolBlend {blendInTime=2, blendOption=Cubic}
    boundBone {boneName=dyn_hair_R_04}
    componentAppliedLinearAccClamp=<3000 3000 3000>
    componentLinearAccScale=<0.3 0.4 0.3>
    componentLinearVelScale=<0.3 0.4 0.3>
    gravityOverride {z=-980}
    physicsBodyDefinitions [{
      boundBone {boneName=dyn_hair_R_04}
      boxExtents=<25 25 25>
      constraintSetup {angularLimitsMax=<0 20 5>, angularLimitsMin=<0 -10 -5>, angularTarget=<1 0 0>}
      localJointOffset=<2 0 0>
     }]
   }
   n19 = Anim Dynamics - Bone: dyn_hair_L_05!(ComponentPose=@n18, bAlphaBoolEnabled:bool=True, GravityOverride:Vector="0,0,-980", AngularSpringConstant:float=20000, ExternalForce="0,0,0") -> (Pose) {
    alphaInputType=Bool
    angularDampingOverride=1
    angularSpringConstant=20000
    bAngularSpring=true
    bOverrideAngularDamping=true
    bUseGravityOverride=true
    gravityScale=0.35
    simulationSpace=RootRelative
    alphaBoolBlend {blendInTime=2, blendOption=Cubic}
    boundBone {boneName=dyn_hair_L_05}
    componentAppliedLinearAccClamp=<3000 3000 3000>
    componentLinearAccScale=<0.3 0.4 0.3>
    componentLinearVelScale=<0.3 0.4 0.3>
    gravityOverride {z=-980}
    physicsBodyDefinitions [{
      boundBone {boneName=dyn_hair_L_05}
      boxExtents=<25 25 25>
      constraintSetup {angularLimitsMax=<0 0 6>, angularLimitsMin=<0 -20 -6>, angularTarget=<1 0 0>}
      localJointOffset=<2 0 0>
     }]
   }
   n20 = Anim Dynamics - Bone: dyn_hair_M_05!(ComponentPose=@n19, bAlphaBoolEnabled:bool=True, GravityOverride:Vector="0,0,-980", AngularSpringConstant:float=20000, ExternalForce="0,0,0") -> (Pose) {
    alphaInputType=Bool
    angularDampingOverride=1
    angularSpringConstant=20000
    bAngularSpring=true
    bOverrideAngularDamping=true
    bUseGravityOverride=true
    gravityScale=0.35
    simulationSpace=RootRelative
    alphaBoolBlend {blendInTime=2, blendOption=Cubic}
    boundBone {boneName=dyn_hair_M_05}
    componentAppliedLinearAccClamp=<3000 3000 3000>
    componentLinearAccScale=<0.3 0.4 0.3>
    componentLinearVelScale=<0.3 0.4 0.3>
    gravityOverride {z=-980}
    physicsBodyDefinitions [{
      boundBone {boneName=dyn_hair_M_05}
      boxExtents=<25 25 25>
      constraintSetup {angularLimitsMax=<0 0 6>, angularLimitsMin=<0 -20 -6>, angularTarget=<1 0 0>}
      localJointOffset=<2 0 0>
     }]
   }
   n21 = Anim Dynamics - Bone: dyn_hair_R_05!(ComponentPose=@n20, bAlphaBoolEnabled:bool=True, GravityOverride:Vector="0,0,-980", AngularSpringConstant:float=20000, ExternalForce="0,0,0") -> (Pose) {
    alphaInputType=Bool
    angularDampingOverride=1
    angularSpringConstant=20000
    bAngularSpring=true
    bOverrideAngularDamping=true
    bUseGravityOverride=true
    gravityScale=0.35
    simulationSpace=RootRelative
    alphaBoolBlend {blendInTime=2, blendOption=Cubic}
    boundBone {boneName=dyn_hair_R_05}
    componentAppliedLinearAccClamp=<3000 3000 3000>
    componentLinearAccScale=<0.3 0.4 0.3>
    componentLinearVelScale=<0.3 0.4 0.3>
    gravityOverride {z=-980}
    physicsBodyDefinitions [{
      boundBone {boneName=dyn_hair_R_05}
      boxExtents=<25 25 25>
      constraintSetup {angularLimitsMax=<0 0 6>, angularLimitsMin=<0 -20 -6>, angularTarget=<1 0 0>}
      localJointOffset=<2 0 0>
     }]
   }
   n22 = Anim Dynamics - Bone: dyn_hair_L_06!(ComponentPose=@n21, bAlphaBoolEnabled:bool=True, GravityOverride:Vector="0,0,-980", AngularSpringConstant:float=20000, ExternalForce="0,0,0") -> (Pose) {
    alphaInputType=Bool
    angularDampingOverride=1
    angularSpringConstant=20000
    bAngularSpring=true
    bOverrideAngularDamping=true
    bUseGravityOverride=true
    gravityScale=0.35
    simulationSpace=RootRelative
    alphaBoolBlend {blendInTime=2, blendOption=Cubic}
    boundBone {boneName=dyn_hair_L_06}
    componentAppliedLinearAccClamp=<3000 3000 3000>
    componentLinearAccScale=<0.3 0.4 0.3>
    componentLinearVelScale=<0.3 0.4 0.3>
    gravityOverride {z=-980}
    physicsBodyDefinitions [{
      boundBone {boneName=dyn_hair_L_06}
      boxExtents=<25 25 25>
      constraintSetup {angularLimitsMax=<0 10 5>, angularLimitsMin=<0 -10 -5>, angularTarget=<1 0 0>}
      localJointOffset=<2 0 0>
     }]
   }
   n23 = Anim Dynamics - Bone: dyn_hair_M_06!(ComponentPose=@n22, bAlphaBoolEnabled:bool=True, GravityOverride:Vector="0,0,-980", AngularSpringConstant:float=20000, ExternalForce="0,0,0") -> (Pose) {
    alphaInputType=Bool
    angularDampingOverride=1
    angularSpringConstant=20000
    bAngularSpring=true
    bOverrideAngularDamping=true
    bUseGravityOverride=true
    gravityScale=0.35
    simulationSpace=RootRelative
    alphaBoolBlend {blendInTime=2, blendOption=Cubic}
    boundBone {boneName=dyn_hair_M_06}
    componentAppliedLinearAccClamp=<3000 3000 3000>
    componentLinearAccScale=<0.3 0.4 0.3>
    componentLinearVelScale=<0.3 0.4 0.3>
    gravityOverride {z=-980}
    physicsBodyDefinitions [{
      boundBone {boneName=dyn_hair_M_06}
      boxExtents=<25 25 25>
      constraintSetup {angularLimitsMax=<0 10 5>, angularLimitsMin=<0 -10 -5>, angularTarget=<1 0 0>}
      localJointOffset=<2 0 0>
     }]
   }
   n24 = Anim Dynamics - Bone: dyn_hair_R_06!(ComponentPose=@n23, bAlphaBoolEnabled:bool=True, GravityOverride:Vector="0,0,-980", AngularSpringConstant:float=20000, ExternalForce="0,0,0") -> (Pose) {
    alphaInputType=Bool
    angularDampingOverride=1
    angularSpringConstant=20000
    bAngularSpring=true
    bOverrideAngularDamping=true
    bUseGravityOverride=true
    gravityScale=0.35
    simulationSpace=RootRelative
    alphaBoolBlend {blendInTime=2, blendOption=Cubic}
    boundBone {boneName=dyn_hair_R_06}
    componentAppliedLinearAccClamp=<3000 3000 3000>
    componentLinearAccScale=<0.3 0.4 0.3>
    componentLinearVelScale=<0.3 0.4 0.3>
    gravityOverride {z=-980}
    physicsBodyDefinitions [{
      boundBone {boneName=dyn_hair_R_06}
      boxExtents=<25 25 25>
      constraintSetup {angularLimitsMax=<0 10 5>, angularLimitsMin=<0 -10 -5>, angularTarget=<1 0 0>}
      localJointOffset=<2 0 0>
     }]
   }
   n25 = "Transform (Modify) Bone - Bone: dyn_Earing_r"(ComponentPose=@n24, Alpha=1, Translation="0,-0.5,0", Rotation="0,0,-29", Scale="1,1,1") {
    rotationMode=BMM_Additive
    rotationSpace=BCS_BoneSpace
    translationMode=BMM_Additive
    translationSpace=BCS_BoneSpace
    alphaBoolBlend {blendOption=Linear}
    boneToModify {boneName=dyn_Earing_r}
    rotation {roll=-29}
    translation {y=-0.5}
   }
   n26 = "Transform (Modify) Bone - Bone: dyn_hair_L_01"!(ComponentPose=@n25, bAlphaBoolEnabled:bool=True, Translation="0,0,8", Rotation="0,0,0") -> (Pose) {
    alphaInputType=Bool
    translationMode=BMM_Additive
    translationSpace=BCS_BoneSpace
    alphaBoolBlend {blendInTime=0.2, blendOption=Linear, blendOutTime=0.5}
    boneToModify {boneName=dyn_hair_L_01}
    translation {z=8}
   }
   n27 = "Transform (Modify) Bone - Bone: dyn_hair_M_01"!(ComponentPose=@n26, bAlphaBoolEnabled:bool=True, Translation="0,0,8", Rotation="0,0,0") -> (Pose) {
    alphaInputType=Bool
    translationMode=BMM_Additive
    translationSpace=BCS_BoneSpace
    alphaBoolBlend {blendInTime=0.2, blendOption=Linear, blendOutTime=0.5}
    boneToModify {boneName=dyn_hair_M_01}
    translation {z=8}
   }
   n28 = "Transform (Modify) Bone - Bone: dyn_hair_R_01"!(ComponentPose=@n27, bAlphaBoolEnabled:bool=True, Translation="0,0,8", Rotation="0,0,0") -> (Pose) {
    alphaInputType=Bool
    translationMode=BMM_Additive
    translationSpace=BCS_BoneSpace
    alphaBoolBlend {blendInTime=0.2, blendOption=Linear, blendOutTime=0.5}
    boneToModify {boneName=dyn_hair_R_01}
    translation {z=8}
   }
   n29 = "Transform (Modify) Bone - Bone: dyn_hair_L_01"(ComponentPose=@n28, bAlphaBoolEnabled=True, Translation="0,0,2", Rotation="5,0,0") {
    alphaInputType=Bool
    rotationMode=BMM_Additive
    rotationSpace=BCS_BoneSpace
    translationMode=BMM_Additive
    translationSpace=BCS_BoneSpace
    alphaBoolBlend {blendInTime=0.2, blendOption=Linear, blendOutTime=0.5}
    boneToModify {boneName=dyn_hair_L_01}
    rotation {pitch=5}
    translation {z=2}
   }
   n30 = "Transform (Modify) Bone - Bone: dyn_hair_M_01"(ComponentPose=@n29, bAlphaBoolEnabled=True, Translation="0,0,2", Rotation="5,0,0") {
    alphaInputType=Bool
    rotationMode=BMM_Additive
    rotationSpace=BCS_BoneSpace
    translationMode=BMM_Additive
    translationSpace=BCS_BoneSpace
    alphaBoolBlend {blendInTime=0.2, blendOption=Linear, blendOutTime=0.5}
    boneToModify {boneName=dyn_hair_M_01}
    rotation {pitch=5}
    translation {z=2}
   }
   n31 = "Transform (Modify) Bone - Bone: dyn_hair_R_01"(ComponentPose=@n30, bAlphaBoolEnabled=True, Translation="0,0,2", Rotation="7,0,0") {
    alphaInputType=Bool
    rotationMode=BMM_Additive
    rotationSpace=BCS_BoneSpace
    translationMode=BMM_Additive
    translationSpace=BCS_BoneSpace
    alphaBoolBlend {blendInTime=0.2, blendOption=Linear, blendOutTime=0.5}
    boneToModify {boneName=dyn_hair_R_01}
    rotation {pitch=7}
    translation {z=2}
   }
   n32 = "Transform (Modify) Bone - Bone: dyn_hair_R_07"!(ComponentPose=@n31, bAlphaBoolEnabled:bool=True, Translation="0,0.85,3", Rotation="0,0,0") -> (Pose) {
    alphaInputType=Bool
    rotationMode=BMM_Additive
    rotationSpace=BCS_BoneSpace
    translationMode=BMM_Additive
    translationSpace=BCS_BoneSpace
    alphaBoolBlend {blendInTime=0.2, blendOption=Linear, blendOutTime=0.5}
    boneToModify {boneName=dyn_hair_R_07}
    translation {y=0.85, z=3}
   }
   n33 = "Transform (Modify) Bone - Bone: dyn_hair_L_07"(ComponentPose=@n32, bAlphaBoolEnabled=True, Translation="0,0,3", Rotation="0,0,0") {
    alphaInputType=Bool
    rotationMode=BMM_Additive
    rotationSpace=BCS_BoneSpace
    translationMode=BMM_Additive
    translationSpace=BCS_BoneSpace
    alphaBoolBlend {blendInTime=0.2, blendOption=Linear, blendOutTime=0.5}
    boneToModify {boneName=dyn_hair_L_07}
    translation {z=3}
   }
   n34 = "Transform (Modify) Bone - Bone: dyn_hair_L_01"(ComponentPose=@n33, bAlphaBoolEnabled=True, Translation="0,0,0", Rotation="-10,-10,0") {
    alphaInputType=Bool
    rotationMode=BMM_Additive
    rotationSpace=BCS_BoneSpace
    translationMode=BMM_Additive
    translationSpace=BCS_BoneSpace
    alphaBoolBlend {blendInTime=0.2, blendOption=Linear, blendOutTime=0.5}
    boneToModify {boneName=dyn_hair_L_01}
    rotation {pitch=-10, yaw=-10}
   }
   n35 = "Transform (Modify) Bone - Bone: dyn_hair_M_01"(ComponentPose=@n34, bAlphaBoolEnabled=True, Translation="0,0,0", Rotation="-10,-10,0") {
    alphaInputType=Bool
    rotationMode=BMM_Additive
    rotationSpace=BCS_BoneSpace
    translationMode=BMM_Additive
    translationSpace=BCS_BoneSpace
    alphaBoolBlend {blendInTime=0.2, blendOption=Linear, blendOutTime=0.5}
    boneToModify {boneName=dyn_hair_M_01}
    rotation {pitch=-10, yaw=-10}
   }
   n36 = "Transform (Modify) Bone - Bone: dyn_hair_R_01"(ComponentPose=@n35, bAlphaBoolEnabled=True, Translation="0,0,0", Rotation="-10,5,0") {
    alphaInputType=Bool
    rotationMode=BMM_Additive
    rotationSpace=BCS_BoneSpace
    translationMode=BMM_Additive
    translationSpace=BCS_BoneSpace
    alphaBoolBlend {blendInTime=0.2, blendOption=Linear, blendOutTime=0.5}
    boneToModify {boneName=dyn_hair_R_01}
    rotation {pitch=-10, yaw=5}
   }
   n37 = "Transform (Modify) Bone - Bone: dyn_hair_L_07"!(ComponentPose=@n36, Alpha:float=1, Translation="2.764,2.28,0", Rotation="57.409,-53.714,0") -> (Pose) {
    rotationMode=BMM_Additive
    rotationSpace=BCS_BoneSpace
    translationMode=BMM_Additive
    translationSpace=BCS_BoneSpace
    alphaBoolBlend {blendOption=Linear}
    boneToModify {boneName=dyn_hair_L_07}
    rotation {pitch=57.409, yaw=-53.714}
    translation=<2.764 2.28>
   }
   n38 = RigidBody!(ComponentPose=@n37, Alpha, OverrideWorldGravity:Vector="0,0,0") -> (Pose) {
    bClampLinearTranslationLimitToRefPose=true
    bOverrideWorldGravity=true
    overridePhysicsAsset=&/F_MED_RoyalAngst_ToteMask_FaceAcc_Physics
    simulationSpace=BaseBoneSpace
    alphaBoolBlend {blendOption=Linear}
    alphaScaleBiasClamp {bInterpResult=true, interpSpeedDecreasing=3, interpSpeedIncreasing=3}
    baseBoneRef {boneName=spine_05}
    componentAppliedLinearAccClamp=<3000 3000 3000>
    simSpaceSettings {
     maxAngularAcceleration=200
     maxAngularVelocity=3
     maxLinearAcceleration=700
     maxLinearVelocity=500
     worldAlpha=1
     externalLinearDragV=<1 1 1>
    }
   }
   Save cached pose 'NoBackblingHair'(Pose=Component To Local(ComponentPose=@n38)) {cacheName=NoBackblingHair}
   n39 = Anim Dynamics - Bone: dyn_hair_M_05(
    ComponentPose=Local To Component(LocalPose=Use cached pose 'Front2backhair'())
    bAlphaBoolEnabled=True
    GravityOverride="0,0,-980"
    AngularSpringConstant=20000
    ExternalForce="0,0,0"
   ) {
    alphaInputType=Bool
    angularDampingOverride=1
    angularSpringConstant=20000
    bAngularSpring=true
    bOverrideAngularDamping=true
    bUseGravityOverride=true
    bUseSphericalLimits=true
    gravityScale=0.35
    simulationSpace=RootRelative
    alphaBoolBlend {blendInTime=2, blendOption=Cubic}
    boundBone {boneName=dyn_hair_M_05}
    componentAppliedLinearAccClamp=<400000 400000 400000>
    componentLinearAccScale=<0.3 0.4 0.3>
    componentLinearVelScale=<0.3 0.4 3>
    gravityOverride {z=-980}
    physicsBodyDefinitions [{
      boundBone {boneName=dyn_hair_M_05}
      boxExtents=<25 25 25>
      constraintSetup {angularLimitsMax=<0 0 10>, angularLimitsMin=<0 -25 -10>, angularTarget=<1 0 0>}
      localJointOffset=<7 0 0>
     }]
    relativeSpaceBone {boneName=root}
    sphericalLimits [{limitRadius=15, drivingBone {boneName=pelvis}, sphereLocalOffset=<0 0 0>}]
   }
   n40 = Anim Dynamics - Bone: dyn_hair_L_05(ComponentPose=@n39, bAlphaBoolEnabled=True, GravityOverride="0,0,-980", AngularSpringConstant=20000, ExternalForce="0,0,0") {
    alphaInputType=Bool
    angularDampingOverride=1
    angularSpringConstant=20000
    bAngularSpring=true
    bOverrideAngularDamping=true
    bUseGravityOverride=true
    bUseSphericalLimits=true
    gravityScale=0.35
    simulationSpace=RootRelative
    alphaBoolBlend {blendInTime=2, blendOption=Cubic}
    boundBone {boneName=dyn_hair_L_05}
    componentAppliedLinearAccClamp=<400000 400000 400000>
    componentLinearAccScale=<0.3 0.4 0.3>
    componentLinearVelScale=<0.3 0.4 3>
    gravityOverride {z=-980}
    physicsBodyDefinitions [{
      boundBone {boneName=dyn_hair_L_05}
      boxExtents=<25 25 25>
      constraintSetup {angularLimitsMax=<0 0 10>, angularLimitsMin=<0 -25 -10>, angularTarget=<1 0 0>}
      localJointOffset=<7 0 0>
     }]
    relativeSpaceBone {boneName=root}
    sphericalLimits [{limitRadius=15, drivingBone {boneName=pelvis}, sphereLocalOffset=<0 0 0>}]
   }
   n41 = Anim Dynamics - Bone: dyn_hair_R_05(ComponentPose=@n40, bAlphaBoolEnabled=True, GravityOverride="0,0,-980", AngularSpringConstant=20000, ExternalForce="0,0,0") {
    alphaInputType=Bool
    angularDampingOverride=1
    angularSpringConstant=20000
    bAngularSpring=true
    bOverrideAngularDamping=true
    bUseGravityOverride=true
    bUseSphericalLimits=true
    gravityScale=0.35
    simulationSpace=RootRelative
    alphaBoolBlend {blendInTime=2, blendOption=Cubic}
    boundBone {boneName=dyn_hair_R_05}
    componentAppliedLinearAccClamp=<400000 400000 400000>
    componentLinearAccScale=<0.3 0.4 0.3>
    componentLinearVelScale=<0.3 0.4 3>
    gravityOverride {z=-980}
    physicsBodyDefinitions [{
      boundBone {boneName=dyn_hair_R_05}
      boxExtents=<25 25 25>
      constraintSetup {angularLimitsMax=<0 0 10>, angularLimitsMin=<0 -25 -10>, angularTarget=<1 0 0>}
      localJointOffset=<7 0 0>
     }]
    relativeSpaceBone {boneName=root}
    sphericalLimits [{limitRadius=15, drivingBone {boneName=pelvis}, sphereLocalOffset=<0 0 0>}]
   }
   n42 = Anim Dynamics - Bone: dyn_hair_M_06(ComponentPose=@n41, bAlphaBoolEnabled=True, GravityOverride="0,0,-980", AngularSpringConstant=20000, ExternalForce="0,0,0") {
    alphaInputType=Bool
    angularDampingOverride=1
    angularSpringConstant=20000
    bAngularSpring=true
    bOverrideAngularDamping=true
    bUseGravityOverride=true
    bUseSphericalLimits=true
    gravityScale=0.35
    alphaBoolBlend {blendInTime=2, blendOption=Cubic}
    boundBone {boneName=dyn_hair_M_06}
    componentAppliedLinearAccClamp=<400000 400000 400000>
    componentLinearAccScale=<0.3 0.4 0.3>
    componentLinearVelScale=<0.3 0.4 3>
    gravityOverride {z=-980}
    physicsBodyDefinitions [{
      boundBone {boneName=dyn_hair_M_06}
      boxExtents=<25 25 25>
      constraintSetup {angularLimitsMax=<0 10 15>, angularLimitsMin=<0 -25 -15>, angularTarget=<1 0 0>}
      localJointOffset=<7 0 0>
     }]
    relativeSpaceBone {boneName=root}
    sphericalLimits [{limitRadius=15, drivingBone {boneName=pelvis}, sphereLocalOffset=<0 0 0>}]
   }
   n43 = Anim Dynamics - Bone: dyn_hair_L_06(ComponentPose=@n42, bAlphaBoolEnabled=True, GravityOverride="0,0,-980", AngularSpringConstant=20000, ExternalForce="0,0,0") {
    alphaInputType=Bool
    angularDampingOverride=1
    angularSpringConstant=20000
    bAngularSpring=true
    bOverrideAngularDamping=true
    bUseGravityOverride=true
    bUseSphericalLimits=true
    gravityScale=0.35
    alphaBoolBlend {blendInTime=2, blendOption=Cubic}
    boundBone {boneName=dyn_hair_L_06}
    componentAppliedLinearAccClamp=<400000 400000 400000>
    componentLinearAccScale=<0.3 0.4 0.3>
    componentLinearVelScale=<0.3 0.4 3>
    gravityOverride {z=-980}
    physicsBodyDefinitions [{
      boundBone {boneName=dyn_hair_L_06}
      boxExtents=<25 25 25>
      constraintSetup {angularLimitsMax=<0 10 15>, angularLimitsMin=<0 -25 -15>, angularTarget=<1 0 0>}
      localJointOffset=<7 0 0>
     }]
    relativeSpaceBone {boneName=root}
    sphericalLimits [{limitRadius=15, drivingBone {boneName=pelvis}, sphereLocalOffset=<0 0 0>}]
   }
   n44 = Anim Dynamics - Bone: dyn_hair_R_06(ComponentPose=@n43, bAlphaBoolEnabled=True, GravityOverride="0,0,-980", AngularSpringConstant=20000, ExternalForce="0,0,0") {
    alphaInputType=Bool
    angularDampingOverride=1
    angularSpringConstant=20000
    bAngularSpring=true
    bOverrideAngularDamping=true
    bUseGravityOverride=true
    bUseSphericalLimits=true
    gravityScale=0.35
    alphaBoolBlend {blendInTime=2, blendOption=Cubic}
    boundBone {boneName=dyn_hair_R_06}
    componentAppliedLinearAccClamp=<400000 400000 400000>
    componentLinearAccScale=<0.3 0.4 0.3>
    componentLinearVelScale=<0.3 0.4 3>
    gravityOverride {z=-980}
    physicsBodyDefinitions [{
      boundBone {boneName=dyn_hair_R_06}
      boxExtents=<25 25 25>
      constraintSetup {angularLimitsMax=<0 10 15>, angularLimitsMin=<0 -25 -15>, angularTarget=<1 0 0>}
      localJointOffset=<7 0 0>
     }]
    relativeSpaceBone {boneName=root}
    sphericalLimits [{limitRadius=15, drivingBone {boneName=pelvis}, sphereLocalOffset=<0 0 0>}]
   }
   Save cached pose 'BackblingHair'(Pose=Component To Local(ComponentPose=@n44)) {cacheName=BackblingHair}
   n45 = "Get Face Mesh "() -> ("Face Mesh ":Skeletal Mesh Component Object Reference)
   n46 = Copy Pose From Mesh() {bUseAttachedParent=true}
   n47 = Copy Pose From Mesh(SourceMeshComponent=Get SkelMesh()) {bUseAttachedParent=true}
   Save cached pose 'Start_Cache'(Pose=Layered blend per bone(
     BasePose=Layered blend per bone(BasePose=@n46, BlendPoses_0=@n47)
     BlendPoses_0=Copy Pose From Mesh(SourceMeshComponent=@n45) {bUseAttachedParent=true}
    ) {layerSetup [{branchFilters [{boneName=FACIAL_R_ForeheadMid}]}]}) {cacheName=Start_Cache}
  ]}
 EventGraph {nodes [
   Event BlueprintUpdateAnimation()
   > n1 = Sequence!(execute) -> (then_0, then_1, then_2:Exec, then_3:Exec, then_4:Exec, then_5:Exec)
   > n2 = Gravity Override State Function DEPRECATED(CharacterPartAnimInstance=Self-Reference(), GameTimeInSeconds=GetGameTimeInSeconds(), GravityOverrideParams=Get Gravity Override Params())
   > n3 = Set Gravity Override(Gravity Override=@n2)
   > n4 = Set GravityOverride AnimDynamics(GravityOverride AnimDynamics:Vector=@n3) -> (Output_Get:Vector)
   n5 = Get bIsSkyDiving()
   n6 = Get bIsJumping()
   n7 = Get bIsFalling()
   > Branch(Condition=AND Boolean(A=OR Boolean(A=@n6, B=@n7), B=NOT Boolean(A=@n5)))
   n8 = Remap Float(Input=Get PawnSpeed2D(), OldRange=(X=0,Y=550), NewRange=(X=0,Y=1), Clamp=true) -> (Output=~)
   n9 = "Clamp (Float)"(Value=Get PawnVelocityZ(), Min=-980, Max=0)
   n10 = float + float(A=Lerp(A=1300, B=float * float(A=@n9, B=-2) {is=title K2Node_PromotableOperator}, Alpha=@n8), B=Lerp(A=0, B=800, Alpha=@n8))
   > Set GravityOverride AnimDynamics(GravityOverride AnimDynamics=vector + vector(A=@n4, B=MakeVector(Y=0, Z=@n10)))
   n11 = GetAnimInstance(self=GetOwnerSkeletalMeshComponent())
   Set IsPlayCelebrateEmote(
    execute=@n1.then_5
    IsPlayCelebrateEmote:double="To Float (Boolean)"(InBool=Montage_IsPlaying(self=@n11, Montage=/BRCosmetics/Animation/Game/MainPlayer/Emotes/Emote_Celebration_Female_M))
   ) -> (Output_Get:double)
   n12 = Cast To FortPlayerAnimInstance(execute=@n1.then_4, Object=GetAnimInstance(self=GetOwnerSkeletalMeshComponent()))
   > Set isPistolEquipped(isPistolEquipped:bool=false) -> (Output_Get:bool)
   > n13 = Switch on EFortWeaponCoreAnimation!(execute, Selection:EFortWeaponCoreAnimation Enum=Get WeaponCoreAnim(self=@n12)) -> (
    Melee:Exec
    Pistol:Exec
    Shotgun:Exec
    PaperBlueprint:Exec
    Rifle:Exec
    MeleeOneHand:Exec
    MachinePistol:Exec
    RocketLauncher:Exec
    GrenadeLauncher:Exec
    GoingCommando:Exec
    AssaultRifle:Exec
    TacticalShotgun:Exec
    SniperRifle:Exec
    TrapPlacement:Exec
    ShoulderLauncher:Exec
    AbilityDecoTool:Exec
    Crossbow:Exec
    C4:Exec
    RemoteControl:Exec
    DualWield:Exec
    AR_BullPup:Exec
    AR_ForwardGrip:Exec
    MedPackPaddles:Exec
    SMG_P90:Exec
    AR_DrumGun:Exec
    Consumable_Small:Exec
    Consumable_Large:Exec
    Balloon:Exec
    MountedTurret:Exec
    CreativeTool:Exec
    ExplosiveBow:Exec
    AshtonIndigo:Exec
    AshtonChicago:Exec
    MeleeDualWield:Exec
    Unarmed:Exec
   )
   n14 = Branch(execute=@n13.Pistol, Condition=Get bIsTargeting())
   Set isPistolEquipped(execute=[@n14.then, @n16.else], isPistolEquipped=true)
   n15 = Branch(execute=@n14.else, Condition=Get bIsRelaxedLevel1() -> (bIsRelaxedLevel1:bool))
   Set isPistolEquipped(execute=[@n15.then, @n16.then], isPistolEquipped=false)
   n16 = Branch(execute=@n15.else, Condition=Get bIsRelaxedLevel2() -> (bIsRelaxedLevel2:bool))
   Set disableUpperLimbsCorrection(
    execute=@n1.then_1
    disableUpperLimbsCorrection:bool=OR Boolean!(A=Get bIsSkydiveDiveMode(), B=Get bIsSkydiveFloating(), C:bool=Get bIsParachuteOpened()) -> (ReturnValue)
   ) -> (Output_Get:bool)
   > Set toggleAnimDynamics(toggleAnimDynamics:bool=NOT Boolean(A=Get isOnZipline() -> (isOnZipline:bool))) -> (Output_Get:bool)
   n17 = Get bIsTacticalSprinting()
   n18 = AND Boolean(A=Get bIsJumping(), B=@n17)
   > Set isTacticalJumping(isTacticalJumping:bool=@n18) -> (Output_Get:bool)
   n19 = Get bIsFalling()
   > Set isTacticalJumpingFalling(isTacticalJumpingFalling:bool=OR Boolean!(A=@n18, B=AND Boolean(A=@n17, B=@n19)) -> (ReturnValue)) -> (Output_Get:bool)
   > Set Head Up(Head Up:float=GetAlphaFromReferenceTransformDelta(SkeletalMeshComponent=Get SkelMesh(), BoneName=spine_04, BaseName=head, InputRange=(X=15,Y=60), TwistAxis="0,0,1")) -> (Output_Get:float)
   n20 = Get Is Backbling Equipped()
   > Set DiveModeNoBB(DiveModeNoBB:bool=OR Boolean(A=Get bIsSkydiveDiveMode(), B=NOT Boolean(A=@n20))) -> (Output_Get:bool)
   n21 = NOT Boolean(A=Get bIsPlayingEmote())
   n22 = Get bIsInFrontEnd()
   > Set FrontEndAlpha(FrontEndAlpha:double="To Float (Boolean)"(InBool=AND Boolean(A=@n21, B=@n22))) -> (Output_Get:double)
   n23 = Branch(execute=@n1.then_2, Condition=OR Boolean(A=Get bIsSkydiveDiveMode(), B=Get bIsSkydiveFloating()))
   > Set angularSpringConstant(angularSpringConstant:double=10000) -> (Output_Get:double)
   n24 = Branch(execute=@n23.else, Condition=Get bIsSprinting())
   > n25 = Branch(Condition=Get bIsTacticalSprinting())
   > Set angularSpringConstant(angularSpringConstant=15000)
   Set angularSpringConstant(execute=@n25.else, angularSpringConstant=17000)
   Set angularSpringConstant(execute=@n24.else, angularSpringConstant=8000)
   n26 = Branch(execute=@n1.then_3, Condition=Get bIsInFrontEnd())
   > Set alphaHairRigidBody float(alphaHairRigidBody float:double=0.45) -> (Output_Get:double)
   n27 = Branch(execute=@n26.else, Condition=Get bIsPlayingMeleeAnim())
   Set alphaHairRigidBody float(execute=[@n27.then, @n29.then], alphaHairRigidBody float=1)
   n28 = Branch(execute=@n27.else, Condition=Get bIsTargeting())
   > Set alphaHairRigidBody float(alphaHairRigidBody float=0.6)
   n29 = Branch(execute=@n28.else, Condition=Get isTacticalJumpingFalling() -> (isTacticalJumpingFalling:bool))
   Set alphaHairRigidBody float(execute=@n29.else, alphaHairRigidBody float=1)
   > Branch(Condition=Get bIsSkyDiving())
   > Set alphaHairRigidBody float(alphaHairRigidBody float=0.7)
   Event BlueprintBeginPlay() {is=title K2Node_Event}
   > Set disableControlRig(disableControlRig:bool=false) -> (Output_Get:bool)
   Event OnCharacterCustomizationLoaded()
   > n30 = Sequence!(execute) -> (then_0, then_1)
   n31 = CheckForBackbling()
   > n32 = Set Is Wearing BP(Is Wearing BP:bool=OR Boolean!(A=@n31, B=@n31.bIsBackblingACape, C:bool=@n31.bIsBackblingLong) -> (ReturnValue)) -> (Output_Get:bool)
   > n33 = Set Is Wearing Cape(Is Wearing Cape=@n31.bIsBackblingLong)
   > Set Is Backbling Equipped(Is Backbling Equipped:bool=OR Boolean!(A=@n33, B=@n32) -> (ReturnValue)) -> (Output_Get:bool)
   n34 = Cast To PlayerPawn_Athena(execute=@n30.then_1, Object=TryGetPawnOwner())
   > "Set Face Mesh "("Face Mesh ":Skeletal Mesh Component Object Reference=GetSkeletalMeshForPartType(self=@n34, PartType=Head)) -> (Output_Get:Skeletal Mesh Component Object Reference)
   Event OnWeaponChanged()
   > Set isAssaultRifleEquipped float(isAssaultRifleEquipped float:double=0) -> (Output_Get:double)
   > n35 = Switch on EFortWeaponCoreAnimation(Selection=Get WeaponCoreAnim!() -> (WeaponCoreAnim))
   Set isAssaultRifleEquipped float(execute=@n35.AssaultRifle, isAssaultRifleEquipped float=1)
   Event BlueprintInitializeAnimation()
   > Set SkelMesh(SkelMesh=GetOwnerSkeletalMeshComponent())
   n36 = Event BlueprintProcessAnimInputEvent()
   > n37 = Branch(Condition=IsNamed(self=@n36.AnimInputEvent, InEventName=aie.Player.Zipline.Entered))
   > Set isOnZipline(isOnZipline:bool=1) -> (Output_Get:bool)
   Branch(execute=@n37.else, Condition=IsNamed(self=@n36.AnimInputEvent, InEventName=aie.Player.Zipline.Exited))
   > Set isOnZipline(isOnZipline=false)
  ]}
]
variables [
 alphaHairRigidBody float {type=number}
 angularSpringConstant {type=number}
 disableControlRig {type=boolean}
 disableUpperLimbsCorrection {type=boolean}
 DiveModeNoBB {type=boolean}
 "Face Mesh " {type=SkeletalMeshComponent}
 FrontEndAlpha {type=number}
 Gravity Override {type=Vector}
 Gravity Override Params {type=GravityOverrideParamsStruct}
 GravityOverride AnimDynamics {type=Vector}
 Head Up {type=number}
 Is Backbling Equipped {type=boolean}
 Is Wearing BP {type=boolean}
 Is Wearing Cape {type=boolean}
 isAssaultRifleEquipped float {type=number}
 isOnZipline {type=boolean}
 isPistolEquipped {type=boolean}
 IsPlayCelebrateEmote {type=number}
 isTacticalJumping {type=boolean}
 isTacticalJumpingFalling {type=boolean}
 SkelMesh {type=SkeletalMeshComponent}
 toggleAnimDynamics {type=boolean}
]
